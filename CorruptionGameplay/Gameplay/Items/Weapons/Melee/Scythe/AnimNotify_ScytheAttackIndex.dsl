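; Anim notify for the Scythe melee weapon that marks which attack hit is playing during an animation.
; On notify, it casts the mesh owner to FortPlayerPawnAthena and its current weapon to B_PowerUp_Scythe.
; It then writes the notify's Index variable into the scythe's HitIndex, returning false if the weapon cast fails.
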(fn Received_Notify
  (params (MeshComp :type Skeletal%20Mesh%20Component%20Object%20Reference) (Animation :type Anim%20Sequence%20Base%20Object%20Reference) (EventReference :type Anim%20Notify%20Event%20Reference%20Structure%20%28by%20ref%29))
  (bind _asfort_player_pawn_athena (Utilities|Casting|CastToFortPlayerPawnAthena (Components|GetOwner MeshComp)))
  (bind _asb_power_up_scythe (Utilities|Casting|CastToB_PowerUp_Scythe (Weapon|GetCurrentWeapon _asfort_player_pawn_athena))
    (:then
      (Class|BPowerUpScythe|SetHitIndex _asb_power_up_scythe (Variables|Default|GetIndex) :!class /CorruptionGameplay/Gameplay/Items/Weapons/Melee/Scythe/B_PowerUp_Scythe.B_PowerUp_Scythe_C)
      (return true))
    (:CastFailed
      (return false))))
