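; Scythe power-up melee weapon that handles impact effect direction and dash-related HUD key prompts.
; On begin play it switches off trail and swing effects, then watches the instigator's dash gameplay tag with wait-for-tag ability tasks.
; Adding or removing the tag swaps visibility between the first three HUD key actions and the last two.
; Impact particles take their direction from an array indexed by an externally settable hit index.
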(event FortWeapon|EventOnPlayImpactFX
  (params (HitResult :type Hit%20Result%20Structure%20%28by%20ref%29) (ImpactPhysicalSurface :type EPhysicalSurface%20Enum :default "SurfaceType_Default") (SpawnedPSC :type FXSystem%20Component%20Object%20Reference))
  (Effects|Components|ParticleSystem|SetVectorParameter SpawnedPSC "ImpactDirection" (Utilities|Array|Get(acopy) (Variables|Default|GetImpactDirections) (Variables|Default|GetHitIndex))))

(event EventTick
  (params (DeltaSeconds :type Float%20%28single-precision%29)))

(event EventBeginPlay
  (|CallParentFunction)
  (Components|Activation|Deactivate (Variables|Default|GetAnimTrail%20Niagara))
  (CallFunction|DeactivateSwingFX)
  (CallFunction|EnableDashPrompts false)
  (Ability|Tasks|WaitGameplayTagAddToActor (Game|GetInstigator) (Variables|Default|GetDashTag)
    (:then
      (Ability|Tasks|WaitGameplayTagRemoveFromActor (Game|GetInstigator) (Variables|Default|GetDashTag)
        (:Removed
          (CallFunction|EnableDashPrompts false))))
    (:Added
      (Development|PrintString "Dash tag added")
      (CallFunction|EnableDashPrompts true))))

(event Custom|ToggleAbilityPrompts
  (params (VisibleStage :type Integer)))

(event Custom|EnableDashPrompts
  (params (Enabled :type Boolean))
  (bind _output_get (Variables|Default|SetArray (Utilities|Array|MakeArray (Utilities|Struct|MakeWeaponHudKeyActionVisibility 0 (not Enabled)) (Utilities|Struct|MakeWeaponHudKeyActionVisibility 1 (not Enabled)) (Utilities|Struct|MakeWeaponHudKeyActionVisibility 2 (not Enabled)) (Utilities|Struct|MakeWeaponHudKeyActionVisibility 3 Enabled) (Utilities|Struct|MakeWeaponHudKeyActionVisibility 4 Enabled))))
  (WeaponHUDData|SetHudKeyActionsVisibility _output_get))

(fn SetHitIndex
  (params (Index :type Integer))
  (Variables|Default|SetHitIndex Index))

(fn ConstructionScript
  (|CallParentFunction))
