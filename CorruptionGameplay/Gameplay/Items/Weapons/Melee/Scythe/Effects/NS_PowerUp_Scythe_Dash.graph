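# Niagara effect for the scythe power-up dash, trailing smoke and small spark-like spots as it moves.
# Both sprite emitters spawn per distance travelled, place particles around points sampled from User.SkeletalMesh, and inherit part of the motion.
# Smoke plays a random-start SubUV flipbook, shading from pale yellow to dark purple.
# GPU-simulated Spots fade from pink-red or magenta to dark, scattered by curl noise, gravity and drag.
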NiagaraSystem /CorruptionGameplay/Gameplay/Items/Weapons/Melee/Scythe/Effects/NS_PowerUp_Scythe_Dash
dependencies {
 usedDataInterfaces [&NiagaraDataInterfaceColorCurve, &NiagaraDataInterfaceCurve]
 usedDynamicInputs []
 usedModules [&SystemState]
}
emitters {
 Smoke {
  emitterUpdateScript [
   &EmitterState()
   &SpawnPerUnit(Module.Use Max Movement Threshold=true, Module.Use Movement Tolerance=false, Use Spawn Probability=false)
   &SpawnBurst_Instantaneous(Spawn Count=3, Use Spawn Probability=false) {enabled=false}
  ]
  particleSpawnScript [
   &InitializeParticle(
    Color Minimum=<rgba 0.25 0.25 0.25 1>
    Color Mode=::NewEnumerator2
    Lifetime Max=0.75
    Lifetime Min=0.3
    Lifetime Mode=::NewEnumerator1
    Position Mode=::NewEnumerator1
    Sprite Rotation Mode=::NewEnumerator2
    Sprite Size Mode=::NewEnumerator1
    Uniform Sprite Size Max=140
    Uniform Sprite Size Min=100
    UsePositionOffset=false
    Write Material Random=true
   ) {module=InitializeParticle001}
   &SkeletalMeshLocation(
    MaskSpawnGroup=false
    Mesh Sampling Type=::NewEnumerator5
    Position Sampling=::NewEnumerator1
    Sampled Position Offset=&MakeVector(X=&RandomRangeFloat(Minimum=-25, Maximum=75), Y=0, Z=-20)
    Skeletal Mesh=$User.SkeletalMesh
    Socket Sampling Mode=::NewEnumerator5
   )
   &SphereLocation(
    MaskSpawnGroup=false
    Offset=&RandomRangeVector(Minimum=<0 0 50>, Maximum=<0 0 -50>)
    Override Seed=false
    Sphere Radius=15
    Surface Expansion Mode=::NewEnumerator1
    Surface Only=true
    UseNonUniformScale=false
    UseOffset=true
    UseSphereOrientation=false
   ) {module=SphereLocation001}
   /SetVariables(Particles.RandomFrame=&RandomRangeInt(Minimum=0, Maximum=63)) {isSetParameters=true}
   &InheritVelocity(
    Inherited Velocity Amount Scale=&VectorFromFloat(Value=0.5)
    Inherited Velocity Speed Limit=200
    Speed Limit=true
    UseSpeedThreshold=false
   )
  ]
  particleUpdateScript [
   &ParticleState()
   &Drag(Drag=&RandomRangeFloat(Minimum=2, Maximum=6), Use Linear Drag=true, Use Rotational Drag=false)
   &SolveForcesAndVelocity(Clamp Velocity=false, Limit Acceleration=false)
   &ScaleColor(
    Scale Alpha=&FloatFromCurve(
     FloatCurve=DataInterface() {propertyValues {Curve {keys [
         {arriveTangent=15.7334, interpMode=RCIM_Cubic, leaveTangent=15.7334, tangentMode=RCTM_User, time=0, value=0}
         {arriveTangent=0.3312, interpMode=RCIM_Cubic, leaveTangent=-0.2993, tangentMode=RCTM_Break, time=0.1, value=1}
         {interpMode=RCIM_Cubic, leaveTangent=-0.6165, time=1, value=0}
        ]}}}
     Scale Curve=&RandomRangeFloat(Minimum=1.25, Maximum=2)
    )
    Scale RGB=&MakeVectorFromLinearColorRGB(Color=&ColorFromCurve(ColorCurve=DataInterface() {propertyValues {
        AlphaCurve {keys [<at 0 1 RCIM_Cubic>]}
        BlueCurve {keys [
          <at 0 0.349>
          {arriveTangent=-0.8941, interpMode=RCIM_Cubic, leaveTangent=-0.8941, time=0.1044, value=0.0762}
          <at 0.3773 0.0116>
          <at 0.7682 0.15>
         ]}
        GreenCurve {keys [
          <at 0 0.9586>
          {arriveTangent=-2.4395, interpMode=RCIM_Cubic, leaveTangent=-2.4395, time=0.1044, value=0.2423}
          <at 0.3773 0.0382>
          <at 0.7682 0>
         ]}
        RedCurve {keys [
          <at 0 1>
          {arriveTangent=-2.3856, interpMode=RCIM_Cubic, leaveTangent=-2.3856, time=0.1044, value=0.6172}
          <at 0.3773 0.0999>
          <at 0.7682 0.0907>
         ]}
       }}))
    ScaleA=true
    ScaleRGB=true
   )
   &ScaleSpriteSize(Scale Factor=&Vector2DFromFloat(Value=&FloatFromCurve(FloatCurve=DataInterface() {propertyValues {Curve {keys [
          {arriveTangent=3.4866, interpMode=RCIM_Cubic, leaveTangent=3.4866, tangentMode=RCTM_User, time=0.0023, value=0}
          {interpMode=RCIM_Cubic, leaveTangent=0.0897, time=1, value=1}
         ]}}})))
   &FN_SubUVAnimation(
    End Frame=&Add_Integer(A=$Particles.RandomFrame, B=63)
    Start Frame=$Particles.RandomFrame
    SubUV Animation Mode:ENiagara_SubUVLookupMode=::NewEnumerator1
    SubUV Curve=DataInterface() {propertyValues {Curve {keys [{arriveTangent=1.9163, interpMode=RCIM_Cubic, leaveTangent=1.9163, tangentMode=RCTM_User, time=0, value=0}, <at 1 1 RCIM_Cubic>]}}}
    SubUV Loop Count=0.5
   )
  ]
  rendererClasses [&NiagaraSpriteRendererProperties]
  renderers [{rendererClass=&NiagaraSpriteRendererProperties}]
 }
 Spots {
  simTarget=GPUComputeSim
  emitterUpdateScript [
   &EmitterState()
   &SpawnPerUnit(
    Module.Use Max Movement Threshold=true
    Module.Use Movement Tolerance=false
    Spawn Probability=0.75
    Spawn Spacing=2
    Use Spawn Probability=true
   )
   &SpawnBurst_Instantaneous(Spawn Count=30, Use Spawn Probability=false) {enabled=false}
  ]
  particleSpawnScript [
   &InitializeParticle(
    Color=<rgba 1 1 1 0.5>
    Lifetime Max=1
    Lifetime Min=0.33
    Lifetime Mode=::NewEnumerator1
    Position Mode=::NewEnumerator1
    Sprite Size Mode=::NewEnumerator1
    Uniform Sprite Size Max=4
    Uniform Sprite Size Min=1
    UsePositionOffset=false
    Write Material Random=true
   )
   &SkeletalMeshLocation(
    MaskSpawnGroup=false
    Mesh Sampling Type=::NewEnumerator5
    Position Sampling=::NewEnumerator1
    Sampled Position Offset=&MakeVector(X=&RandomRangeFloat(Minimum=-40, Maximum=85), Y=0, Z=-20)
    Skeletal Mesh=$User.SkeletalMesh
    Socket Sampling Mode=::NewEnumerator5
   )
   &SphereLocation(
    MaskSpawnGroup=false
    Override Seed=false
    Sphere Radius=5
    Surface Expansion Mode=::NewEnumerator1
    Surface Only=true
    UseNonUniformScale=false
    UseOffset=false
    UseSphereOrientation=false
   )
   &InheritVelocity(
    Inherited Velocity Amount Scale=&VectorFromFloat(Value=-0.5)
    Inherited Velocity Speed Limit=-150
    Speed Limit=true
    UseSpeedThreshold=false
   )
   &AddVelocity(Velocity=&Multiply_VectorByFloat(Vector=$Output.SphereLocation.SphereVector, Float=&RandomRangeFloat(Minimum=15, Maximum=80)))
  ]
  particleUpdateScript [
   &ParticleState()
   &ScaleColor(
    Scale Alpha=&FloatFromCurve(FloatCurve=DataInterface() {propertyValues {Curve {keys [
         {tangentMode=RCTM_User, time=0, value=1}
         {arriveTangent=-0.0405, interpMode=RCIM_Cubic, leaveTangent=-0.0405, tangentMode=RCTM_User, time=0.33, value=1}
         {interpMode=RCIM_Cubic, leaveTangent=-8.0089, time=1, value=0}
        ]}}})
    Scale RGB=&MakeCustomVectorFromBool(
     Bool=&RandomBool()
     True Vector=&Lerp_Vector(
      A=&MakeVectorFromLinearColorRGB(Color=<rgba 1 0.5158 0.5094 1>)
      B=&MakeVectorFromLinearColorRGB(Color=<rgba 0.2292 0.0262 0.066 1>)
      Alpha=&ClampFloat(Float=&Multiply_Float(A=$Particles.NormalizedAge, B=2.5))
     )
     False Vector=&Lerp_Vector(
      A=&MakeVectorFromLinearColorRGB(Color=<rgba 0.735 0 1 1>)
      B=&MakeVectorFromLinearColorRGB(Color=<rgba 0.008 0 0.0685 1>)
      Alpha=&ClampFloat(Float=&Multiply_Float(A=$Particles.NormalizedAge, B=2.5))
     )
    )
    ScaleA=true
    ScaleRGB=true
   )
   &GravityForce(Gravity=&RandomRangeVector(Minimum=<0 0 0>, Maximum=<0 0 -100>))
   &CurlNoiseForce(
    Noise Frequency=15
    Noise Strength=&Multiply_Float(
     A=&RandomRangeFloat(Minimum=1000, Maximum=1500)
     B=&FloatFromCurve(FloatCurve=DataInterface() {propertyValues {Curve {keys [<at 0 0>,<at 0.2 1>]}}}, Scale Curve=0.5)
    )
    Pan Noise=true
   )
   &Drag(Drag=&RandomRangeFloat(Minimum=2, Maximum=6), Use Linear Drag=true, Use Rotational Drag=false)
   &SolveForcesAndVelocity(Clamp Velocity=false, Limit Acceleration=false)
   &ScaleSpriteSize(Scale Factor=&Vector2DFromFloat(Value=&FloatFromCurve(FloatCurve=DataInterface() {propertyValues {Curve {keys [
          {tangentMode=RCTM_User, time=0, value=1}
          {arriveTangent=-0.0405, interpMode=RCIM_Cubic, leaveTangent=-0.0405, tangentMode=RCTM_User, time=0.7009, value=1}
          {arriveTangent=-8.0089, interpMode=RCIM_Cubic, leaveTangent=-8.0089, tangentMode=RCTM_User, time=1, value=0}
         ]}}})))
  ]
  rendererClasses [&NiagaraSpriteRendererProperties]
  renderers [{rendererClass=&NiagaraSpriteRendererProperties}]
 }
}
summary {
 systemName=NS_PowerUp_Scythe_Dash
 emitters [
  {emitterName=Smoke, rendererClasses [&NiagaraSpriteRendererProperties]}
  {emitterName=Spots, simTarget=GPUComputeSim, rendererClasses [&NiagaraSpriteRendererProperties]}
 ]
 userVariables [
  User.bIsFrontEndPreview {type=NiagaraBool}
  User.RibbonIDOffset {type=NiagaraInt32}
  User.SkeletalMesh {
   type=NiagaraDataInterfaceSkeletalMesh
   defaultValue {
    struct=NiagaraExt_VariableValue_DataInterface
    value {dataInterface=&self:NiagaraDataInterfaceSkeletalMesh_0}
   }
  }
  User.WeaponTrailOffset {type=NiagaraBool}
 ]
}
