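# Niagara system for the impact burst of the scythe dash power-up: a brief warm flash with smoke and drifting specks.
# Single-particle Flare, Glow and Light emitters form the flash, with extra flare and glow color parameters dimmed by time of day.
# Smoke bursts outward with random-start SubUV animation, while delayed GPU Spots swirl under curl noise in red or violet.
# The Par spark emitter is disabled, and no user parameters are exposed.
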NiagaraSystem /CorruptionGameplay/Gameplay/Items/Weapons/Melee/Scythe/Effects/NS_PowerUp_Scythe_Dash_Impact
dependencies {
 usedDataInterfaces [&NiagaraDataInterfaceColorCurve, &NiagaraDataInterfaceCurve, &NiagaraDataInterfaceVector2DCurve]
 usedDynamicInputs [&Vector2DFromCurve]
 usedModules [&SystemState]
}
emitters {
 Flare {
  emitterUpdateScript [
   &EmitterState(Life Cycle Mode=::NewEnumerator1, Loop Behavior=::NewEnumerator1, Loop Duration=1, UseLoopDelay=false)
   &SpawnBurst_Instantaneous(Spawn Count=1, Use Spawn Probability=false)
  ]
  particleSpawnScript [
   &InitializeParticle(Lifetime=0.3, Sprite Size Mode=::NewEnumerator3, Uniform Sprite Size=200, UsePositionOffset=false, Write Material Random=true)
   /SetVariables(Particles.FlareColor=<rgba 6 2.3553 0.7406 1>) {isSetParameters=true}
   /SetVariables(Particles.FlareSize:Vector2f=&Vector2DFromFloat(Value=200)) {isSetParameters=true}
   &CameraOffset(Camera Offset Amount=30)
  ]
  particleUpdateScript [
   &ParticleState(Kill Particles When Lifetime Has Elapsed=false, Let Infinitely Lived Particles Die When Emitter Deactivates=true)
   &ScaleColor(
    Scale Alpha=&FloatFromCurve(FloatCurve=DataInterface() {propertyValues {Curve {keys [<at 0.25 1 RCIM_Cubic>,<at 1 0>]}}})
    Scale RGB=&Multiply_Vector(
     A=<1 1 1>
     B=&MakeVectorFromLinearColorRGB(Color=&ColorFromCurve(ColorCurve=DataInterface() {propertyValues {
         AlphaCurve {keys [<at 0 1 RCIM_Cubic>]}
         BlueCurve {keys [<at 0 0.2307>, {arriveTangent=-0.2056, interpMode=RCIM_Cubic, leaveTangent=-0.2056, time=0.4327, value=0.0762}, <at 1 0.0252>]}
         GreenCurve {keys [<at 0 0.956>, {arriveTangent=-0.9523, interpMode=RCIM_Cubic, leaveTangent=-0.9523, time=0.4327, value=0.2423}, <at 1 0.0037>]}
         RedCurve {keys [<at 0 1>, {arriveTangent=-0.9533, interpMode=RCIM_Cubic, leaveTangent=-0.9533, time=0.4327, value=0.6172}, <at 1 0.0467>]}
        }}))
    )
    ScaleA=true
    ScaleRGB=true
   ) {module=ScaleColor002}
   &ScaleColor(
    Scale Alpha=&FNTimeOfDayScalar(Night Scale=0.5, Morning Scale=0.3, Evening Scale=0.5)
    Scale RGB=&VectorFromFloat(Value=&RandomRangeFloat(Minimum=0, Maximum=2, Evaluation Type=::NewEnumerator1))
    ScaleA=true
    ScaleRGB=true
   ) {enabled=false}
   &ScaleSpriteSize(Scale Factor=&Vector2DFromFloat(Value=&Multiply_Float(
      A=&RandomRangeFloat(Minimum=0.5, Maximum=1, Evaluation Type=::NewEnumerator1)
      B=&FloatFromCurve(FloatCurve=DataInterface() {propertyValues {Curve {keys [<at 0 0>,<at 0.15 1>,<at 1 0>]}}})
     )))
   /SetVariables(Particles.Color001:LinearColor=&MultiplyLinearColor(
     Linear Color=$Particles.FlareColor
     Scale Factor=&MakeFloat4FromVectorAndFloat(
      "Vector (RGB)"=&VectorFromFloat(Value=&RandomRangeFloat(Minimum=0, Maximum=2, Evaluation Type=::NewEnumerator1))
      "Float (Alpha)"=&FNTimeOfDayScalar(Night Scale=0.5, Morning Scale=0.3, Evening Scale=0.5)
     )
    )) {isSetParameters=true}
   /SetVariables(Particles.SpriteSize001:Vector2f=&Multiply_Vector2(
     A=$Particles.FlareSize
     B=&Vector2DFromFloat(Value=&Multiply_Float(
       A=&RandomRangeFloat(Minimum=0.5, Maximum=1, Evaluation Type=::NewEnumerator1)
       B=&FloatFromCurve(FloatCurve=DataInterface() {propertyValues {Curve {keys [<at 0 0>,<at 0.15 1>,<at 1 0>]}}})
      ))
    )) {isSetParameters=true}
  ]
  rendererClasses [&NiagaraSpriteRendererProperties]
  renderers [
   {rendererClass=&NiagaraSpriteRendererProperties}
   {rendererClass=&NiagaraSpriteRendererProperties, rendererIndex=1}
  ]
 }
 Glow {
  emitterUpdateScript [
   &EmitterState(Life Cycle Mode=::NewEnumerator1, Loop Behavior=::NewEnumerator1, Loop Duration=1, UseLoopDelay=false)
   &SpawnBurst_Instantaneous(Spawn Count=1, Use Spawn Probability=false)
  ]
  particleSpawnScript [&InitializeParticle(Lifetime=0.2, Sprite Size Mode=::NewEnumerator3, Uniform Sprite Size=150, UsePositionOffset=false, Write Material Random=true)]
  particleUpdateScript [
   &ParticleState()
   &ScaleSpriteSize(Scale Factor=&Vector2DFromFloat(Value=&FloatFromCurve(
      FloatCurve=DataInterface() {propertyValues {Curve {keys [
          {arriveTangent=13.1625, interpMode=RCIM_Cubic, leaveTangent=13.1625, tangentMode=RCTM_User, time=0, value=0}
          {arriveTangent=0.0473, interpMode=RCIM_Cubic, leaveTangent=0.0473, tangentMode=RCTM_User, time=0.1, value=1}
          {arriveTangent=-0.0078, interpMode=RCIM_Cubic, leaveTangent=-0.0078, tangentMode=RCTM_User, time=0.4, value=1}
          {arriveTangent=-2.5609, interpMode=RCIM_Cubic, leaveTangent=-2.5609, tangentMode=RCTM_User, time=1, value=0}
         ]}}}
      Scale Curve=&RandomRangeFloat(Minimum=0.7, Maximum=1, Evaluation Type=::NewEnumerator1)
     )))
   &ScaleColor(
    Linear Color Curve=DataInterface() {propertyValues {
      AlphaCurve {keys [<at 0 1 RCIM_Cubic>]}
      BlueCurve {keys [<at 0 0.2>, {arriveTangent=0.6172, interpMode=RCIM_Cubic, leaveTangent=0.6172, time=0.2134, value=0.0926}, <at 0.486 0.5>, <at 1 0.15>]}
      GreenCurve {keys [<at 0 0.9492>, {arriveTangent=-1.9529, interpMode=RCIM_Cubic, leaveTangent=-1.9529, time=0.2134, value=0.2944}, <at 0.486 0>, <at 1 0>]}
      RedCurve {keys [<at 0 1>, {arriveTangent=-1.4343, interpMode=RCIM_Cubic, leaveTangent=-1.4343, time=0.2134, value=0.75}, <at 0.486 0.3029>, <at 1 0.0908>]}
     }}
    Scale Mode=::NewEnumerator2
   ) {module=ScaleColor001}
   &ScaleColor(
    Color Value To Scale=$Particles.Color
    Scale Alpha=&FloatFromCurve(
     FloatCurve=DataInterface() {propertyValues {Curve {keys [
         {arriveTangent=0.4005, interpMode=RCIM_Cubic, leaveTangent=0.4005, tangentMode=RCTM_User, time=0, value=0}
         {arriveTangent=0.0256, interpMode=RCIM_Cubic, leaveTangent=0.0256, tangentMode=RCTM_User, time=0.1, value=1}
         {arriveTangent=0.01, interpMode=RCIM_Cubic, leaveTangent=0.01, tangentMode=RCTM_User, time=0.5, value=1}
         <at 1 0 RCIM_Cubic>
        ]}}}
     Scale Curve=0.75
    )
    Scale RGB=<3 3 3>
    ScaleA=true
    ScaleRGB=true
   )
   /SetVariables(
    Particles.GlowColor2:LinearColor=&MakeLinearColorFromVectorAndFloat(
     "Vector (RGB)"=&MakeVectorFromLinearColorRGB(Color=<rgba 3 1.1776 0.3703 1>)
     "Float (Alpha)"=&Multiply_Float(
      A=0.2
      B=&FloatFromCurve(
       FloatCurve=DataInterface() {propertyValues {Curve {keys [
           {arriveTangent=0.4005, interpMode=RCIM_Cubic, leaveTangent=0.4005, tangentMode=RCTM_User, time=0, value=0}
           {arriveTangent=0.0256, interpMode=RCIM_Cubic, leaveTangent=0.0256, tangentMode=RCTM_User, time=0.2, value=1}
           {arriveTangent=0.01, interpMode=RCIM_Cubic, leaveTangent=0.01, tangentMode=RCTM_User, time=0.7, value=1}
           <at 1 0 RCIM_Cubic>
          ]}}}
       Scale Curve=&FNTimeOfDayScalar(Night Scale=0.3, Morning Scale=0.2, Evening Scale=0.5)
      )
     )
    )
    Particles.GlowSize2:Vector2f=&Vector2DFromFloat(Value=&Multiply_Float(
      A=150
      B=&FloatFromCurve(
       FloatCurve=DataInterface() {propertyValues {Curve {keys [
           {arriveTangent=0.4346, interpMode=RCIM_Cubic, leaveTangent=0.4346, tangentMode=RCTM_User, time=0, value=0}
           {arriveTangent=0.0473, interpMode=RCIM_Cubic, leaveTangent=0.0473, tangentMode=RCTM_User, time=0.2, value=1}
           {arriveTangent=-0.0078, interpMode=RCIM_Cubic, leaveTangent=-0.0078, tangentMode=RCTM_User, time=0.5, value=1}
           <at 1 0 RCIM_Cubic>
          ]}}}
       Scale Curve=&RandomRangeFloat(Minimum=0.7, Maximum=1, Evaluation Type=::NewEnumerator1)
      )
     ))
   ) {isSetParameters=true}
  ]
  rendererClasses [&NiagaraSpriteRendererProperties]
  renderers [{rendererClass=&NiagaraSpriteRendererProperties}]
 }
 Light {
  emitterUpdateScript [
   &EmitterState(Life Cycle Mode=::NewEnumerator1, Loop Behavior=::NewEnumerator1, Loop Duration=1, UseLoopDelay=false)
   &SpawnBurst_Instantaneous(Spawn Count=1, Use Spawn Probability=false)
  ]
  particleSpawnScript [&InitializeParticle(Lifetime=0.25, UsePositionOffset=false, Write Material Random=true)]
  particleUpdateScript [
   &ParticleState()
   &ScaleColor(
    Scale Alpha=&FloatFromCurve(
     FloatCurve=DataInterface() {propertyValues {Curve {keys [
         {arriveTangent=18.1327, interpMode=RCIM_Cubic, leaveTangent=18.1327, tangentMode=RCTM_User, time=0, value=0}
         {interpMode=RCIM_Cubic, tangentMode=RCTM_User, time=0.15, value=1}
         {interpMode=RCIM_Cubic, tangentMode=RCTM_User, time=0.66, value=1}
         {arriveTangent=-6.8655, interpMode=RCIM_Cubic, leaveTangent=-6.8655, tangentMode=RCTM_User, time=1, value=0}
        ]}}}
     Scale Curve=3
    )
    Scale RGB=&MakeVectorFromLinearColorRGB(Color=&ColorFromCurve(ColorCurve=DataInterface() {propertyValues {
        AlphaCurve {keys [<at 0 1 RCIM_Cubic>]}
        BlueCurve {keys [<at 0 0.0252>, <at 0.1977 0.0762>, {interpMode=RCIM_Cubic, leaveTangent=-0.1031, time=0.4795, value=0.0762}, <at 0.7285 0.0233>]}
        GreenCurve {keys [<at 0 0.0037>, <at 0.1977 0.2423>, {interpMode=RCIM_Cubic, leaveTangent=-0.4728, time=0.4795, value=0.2423}, <at 0.7285 0>]}
        RedCurve {keys [<at 0 0.0467>, <at 0.1977 0.6172>, {interpMode=RCIM_Cubic, leaveTangent=-1.1133, time=0.4795, value=0.6172}, <at 0.7285 0.0467>]}
       }}))
    ScaleA=true
    ScaleRGB=true
   )
  ]
  rendererClasses [&NiagaraLightRendererProperties]
  renderers [{rendererClass=&NiagaraLightRendererProperties, rendererIndex=0}]
 }
 Par {
  bEnabled=false
  emitterUpdateScript [
   &EmitterState(Life Cycle Mode=::NewEnumerator1, Loop Behavior=::NewEnumerator1, Loop Delay=0.066, Loop Duration=1, UseLoopDelay=true)
   &SpawnBurst_Instantaneous(Spawn Count=25, Use Spawn Probability=false)
  ]
  particleSpawnScript [
   &InitializeParticle(
    Color=<rgba 5 5 5 1>
    Lifetime Max=0.16
    Lifetime Min=0.1
    Lifetime Mode=::NewEnumerator1
    Sprite Size Max=<3 6>
    Sprite Size Min=<2 1>
    Sprite Size Mode=::NewEnumerator2
    UsePositionOffset=false
    Write Material Random=true
   )
   &SphereLocation(
    MaskSpawnGroup=false
    Override Seed=false
    Sphere Radius=10
    Surface Expansion Mode=::NewEnumerator2
    Surface Only=true
    Surface Only Band Thickness=1
    UseNonUniformScale=false
    UseOffset=false
    UseSphereOrientation=false
   )
   &AddVelocity(
    Coordinate Space=::Simulation
    Velocity=&Multiply_VectorByFloat(Vector=$Output.SphereLocation.SphereVector, Float=&RandomRangeFloat(Minimum=40, Maximum=60))
   )
   &CameraOffset(Camera Offset Amount=10)
  ]
  particleUpdateScript [
   &ParticleState()
   &Drag(Drag=&RandomRangeFloat(Minimum=1, Maximum=3), Use Linear Drag=true, Use Rotational Drag=false)
   &SolveForcesAndVelocity(Clamp Velocity=false, Limit Acceleration=false)
   &ScaleColor(
    Scale Alpha=&FloatFromCurve(FloatCurve=DataInterface() {propertyValues {Curve {keys [<at 0.66 1 RCIM_Cubic>,<at 1 0 RCIM_Cubic>]}}}, Scale Curve=0.5)
    Scale RGB=&VectorFromFloat(Value=&FloatFromCurve(
      FloatCurve=DataInterface() {propertyValues {Curve {keys [<at 0.5 2 RCIM_Cubic>,<at 1 0.5 RCIM_Cubic>]}}}
      Scale Curve=&FNTimeOfDayScalar(Night Scale=1, Morning Scale=1, Evening Scale=1)
     ))
    ScaleA=true
    ScaleRGB=true
   ) {module=ScaleColor002}
   &ScaleColor(
    Linear Color Curve=DataInterface() {propertyValues {
      AlphaCurve {keys [<at 0 1 RCIM_Cubic>]}
      BlueCurve {keys [<at 0 1>, {interpMode=RCIM_Cubic, leaveTangent=-0.85, time=0.7675, value=1}]}
      GreenCurve {keys [<at 0 0>,<at 0.7675 0.125 RCIM_Cubic>]}
      RedCurve {keys [<at 0 0.7379>, {interpMode=RCIM_Cubic, leaveTangent=-0.6471, time=0.7675, value=0.2708}]}
     }}
    Scale Mode=::NewEnumerator2
   ) {module=ScaleColor001}
   &ScaleSpriteSizeBySpeed(Max Scale Factor=<1 15>, Source Velocity:Vector3f=&Multiply_Vector(A=<0 0.01 1>, B=$Particles.Velocity))
   &ScaleSpriteSize(Uniform Curve Sprite Scale=DataInterface() {propertyValues {Curve {keys [{interpMode=RCIM_Cubic, tangentMode=RCTM_User, time=0, value=1}, {interpMode=RCIM_Cubic, tangentMode=RCTM_User, time=1, value=0}]}}})
  ]
  rendererClasses [&NiagaraSpriteRendererProperties]
  renderers [{rendererClass=&NiagaraSpriteRendererProperties}]
 }
 Smoke {
  emitterUpdateScript [
   &EmitterState(Life Cycle Mode=::NewEnumerator1, Loop Behavior=::NewEnumerator1, Loop Duration=1, UseLoopDelay=true)
   &SpawnBurst_Instantaneous(Spawn Count=5, Use Spawn Probability=false)
  ]
  particleSpawnScript [
   &InitializeParticle(
    Color Minimum=<rgba 0.25 0.25 0.25 1>
    Color Mode=::NewEnumerator2
    Lifetime Max=0.75
    Lifetime Min=0.6
    Lifetime Mode=::NewEnumerator1
    Sprite Rotation Mode=::NewEnumerator2
    Sprite Size Mode=::NewEnumerator1
    Uniform Sprite Size Max=100
    Uniform Sprite Size Min=75
    UsePositionOffset=false
    Write Material Random=true
   ) {module=InitializeParticle001}
   /SetVariables(Particles.RandomFrame=&RandomRangeInt(Minimum=0, Maximum=63)) {isSetParameters=true}
   &SphereLocation(
    MaskSpawnGroup=false
    Override Seed=false
    Sphere Radius=15
    Surface Expansion Mode=::NewEnumerator1
    Surface Only=true
    Surface Only Band Thickness=0.25
    UseNonUniformScale=false
    UseOffset=false
    UseSphereOrientation=false
   )
   &AddVelocity(
    Coordinate Space=::Simulation
    Velocity=&Multiply_VectorByFloat(Vector=$Output.SphereLocation.SphereNormal, Float=100)
   )
  ]
  particleUpdateScript [
   &ParticleState()
   &Drag(Drag=&RandomRangeFloat(Minimum=2, Maximum=6), Use Linear Drag=true, Use Rotational Drag=false)
   &SolveForcesAndVelocity(Clamp Velocity=false, Limit Acceleration=false)
   &ScaleColor(
    Scale Alpha=&FloatFromCurve(
     FloatCurve=DataInterface() {propertyValues {Curve {keys [
         {arriveTangent=15.7334, interpMode=RCIM_Cubic, leaveTangent=15.7334, tangentMode=RCTM_User, time=0, value=0}
         {arriveTangent=0.3312, interpMode=RCIM_Cubic, leaveTangent=-0.2993, tangentMode=RCTM_Break, time=0.1, value=1}
         {interpMode=RCIM_Cubic, leaveTangent=-0.6165, time=1, value=0}
        ]}}}
     Scale Curve=&RandomRangeFloat(Minimum=1.25, Maximum=2)
    )
    Scale RGB=&MakeVectorFromLinearColorRGB(Color=&ColorFromCurve(ColorCurve=DataInterface() {propertyValues {
        AlphaCurve {keys [<at 0 1 RCIM_Cubic>]}
        BlueCurve {keys [
          <at 0 0.349>
          {arriveTangent=-0.8941, interpMode=RCIM_Cubic, leaveTangent=-0.8941, time=0.1044, value=0.0762}
          <at 0.3773 0.0116>
          <at 0.7682 0.15>
         ]}
        GreenCurve {keys [
          <at 0 0.9586>
          {arriveTangent=-2.4395, interpMode=RCIM_Cubic, leaveTangent=-2.4395, time=0.1044, value=0.2423}
          <at 0.3773 0.0382>
          <at 0.7682 0>
         ]}
        RedCurve {keys [
          <at 0 1>
          {arriveTangent=-2.3856, interpMode=RCIM_Cubic, leaveTangent=-2.3856, time=0.1044, value=0.6172}
          <at 0.3773 0.0999>
          <at 0.7682 0.0908>
         ]}
       }}))
    ScaleA=true
    ScaleRGB=true
   ) {module=ScaleColor001}
   &ScaleSpriteSize(Scale Factor=&Vector2DFromFloat(Value=&FloatFromCurve(FloatCurve=DataInterface() {propertyValues {Curve {keys [
          {arriveTangent=4.1785, interpMode=RCIM_Cubic, leaveTangent=4.1785, tangentMode=RCTM_User, time=0, value=0}
          {arriveTangent=0.9995, interpMode=RCIM_Cubic, leaveTangent=0.9995, tangentMode=RCTM_User, time=0.2223, value=0.6825}
          {arriveTangent=-0.0405, interpMode=RCIM_Cubic, leaveTangent=-0.0405, tangentMode=RCTM_User, time=1, value=1}
         ]}}})))
   &FN_SubUVAnimation(
    End Frame=&Add_Integer(A=$Particles.RandomFrame, B=63)
    Start Frame=$Particles.RandomFrame
    SubUV Animation Mode:ENiagara_SubUVLookupMode=::NewEnumerator1
    SubUV Curve=DataInterface() {propertyValues {Curve {keys [{arriveTangent=1.9163, interpMode=RCIM_Cubic, leaveTangent=1.9163, tangentMode=RCTM_User, time=0, value=0}, <at 1 1 RCIM_Cubic>]}}}
    SubUV Loop Count=0.5
   )
  ]
  rendererClasses [&NiagaraSpriteRendererProperties]
  renderers [{rendererClass=&NiagaraSpriteRendererProperties}]
 }
 Spots {
  simTarget=GPUComputeSim
  emitterUpdateScript [
   &EmitterState(Life Cycle Mode=::NewEnumerator1, Loop Behavior=::NewEnumerator1, Loop Delay=0.13, Loop Duration=1, UseLoopDelay=true)
   &SpawnBurst_Instantaneous(Spawn Count=100, Use Spawn Probability=false)
  ]
  particleSpawnScript [
   &InitializeParticle(
    Color=<rgba 1 1 1 0.5>
    Lifetime Max=0.66
    Lifetime Min=0.33
    Lifetime Mode=::NewEnumerator1
    Sprite Size Mode=::NewEnumerator1
    Uniform Sprite Size Max=3
    Uniform Sprite Size Min=1
    UsePositionOffset=false
    Write Material Random=true
   )
   &SphereLocation(
    Hemisphere X=true
    MaskSpawnGroup=false
    Override Seed=false
    Sphere Radius=20
    Surface Only=false
    UseNonUniformScale=false
    UseOffset=false
    UseSphereOrientation=false
   )
   &AddVelocity(
    Coordinate Space=::Simulation
    Velocity=&Multiply_VectorByFloat(Vector=$Output.SphereLocation.SphereVector, Float=&RandomRangeFloat(Minimum=10, Maximum=15))
   )
   &AddVelocity(
    Coordinate Space=::World
    Velocity=&RandomRangeVector(Minimum=<0 0 400>, Maximum=<0 0 800>, Evaluation Type=::NewEnumerator1)
   ) {enabled=false, module=AddVelocity001}
  ]
  particleUpdateScript [
   &ParticleState()
   &ScaleColor(
    Scale Alpha=&FloatFromCurve(FloatCurve=DataInterface() {propertyValues {Curve {keys [
         {tangentMode=RCTM_User, time=0, value=1}
         {arriveTangent=-0.0405, interpMode=RCIM_Cubic, leaveTangent=-0.0405, tangentMode=RCTM_User, time=0.5, value=1}
         {interpMode=RCIM_Cubic, leaveTangent=-8.0089, time=1, value=0}
        ]}}})
    Scale RGB=&MakeCustomVectorFromBool(
     Bool=&RandomBool()
     True Vector=&Lerp_Vector(
      A=&MakeVectorFromLinearColorRGB(Color=<rgba 1 0.5158 0.5094 1>)
      B=&MakeVectorFromLinearColorRGB(Color=<rgba 0.2292 0.0262 0.066 1>)
      Alpha=&ClampFloat(Float=&Multiply_Float(A=$Particles.NormalizedAge, B=2.5))
     )
     False Vector=&Lerp_Vector(
      A=&MakeVectorFromLinearColorRGB(Color=<rgba 0.7379 0 1 1>)
      B=&MakeVectorFromLinearColorRGB(Color=<rgba 0.008 0 0.0685 1>)
      Alpha=&ClampFloat(Float=&Multiply_Float(A=$Particles.NormalizedAge, B=2.5))
     )
    )
    ScaleA=true
    ScaleRGB=true
   )
   &GravityForce(Gravity=&RandomRangeVector(Minimum=<0 0 0>, Maximum=<0 0 -100>))
   &CurlNoiseForce(
    Noise Frequency=25
    Noise Strength=&Multiply_Float(
     A=&RandomRangeFloat(Minimum=1000, Maximum=1500)
     B=&FloatFromCurve(FloatCurve=DataInterface() {propertyValues {Curve {keys [<at 0 0>,<at 0.2 1>]}}})
    )
    Pan Noise=true
   )
   &Drag(Drag=&RandomRangeFloat(Minimum=2, Maximum=6), Use Linear Drag=true, Use Rotational Drag=false)
   &SolveForcesAndVelocity(Clamp Velocity=false, Limit Acceleration=false)
   &ScaleSpriteSize(Scale Factor=&Vector2DFromFloat(Value=&FloatFromCurve(FloatCurve=DataInterface() {propertyValues {Curve {keys [
          {tangentMode=RCTM_User, time=0, value=1}
          {arriveTangent=-0.0405, interpMode=RCIM_Cubic, leaveTangent=-0.0405, tangentMode=RCTM_User, time=0.7009, value=1}
          {arriveTangent=-8.0089, interpMode=RCIM_Cubic, leaveTangent=-8.0089, tangentMode=RCTM_User, time=1, value=0}
         ]}}})))
  ]
  rendererClasses [&NiagaraSpriteRendererProperties]
  renderers [{rendererClass=&NiagaraSpriteRendererProperties}]
 }
}
summary {
 systemName=NS_PowerUp_Scythe_Dash_Impact
 emitters [
  {bEnabled=false, emitterName=Par, rendererClasses [&NiagaraSpriteRendererProperties]}
  {emitterName=Flare, rendererClasses [&NiagaraSpriteRendererProperties]}
  {emitterName=Glow, rendererClasses [&NiagaraSpriteRendererProperties]}
  {emitterName=Light, rendererClasses [&NiagaraLightRendererProperties]}
  {emitterName=Smoke, rendererClasses [&NiagaraSpriteRendererProperties]}
  {emitterName=Spots, simTarget=GPUComputeSim, rendererClasses [&NiagaraSpriteRendererProperties]}
 ]
 userVariables []
}
