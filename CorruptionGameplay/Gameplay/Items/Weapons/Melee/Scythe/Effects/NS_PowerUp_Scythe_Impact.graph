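# Niagara impact effect for the power-up scythe melee weapon, built from eleven short burst emitters with slightly staggered timing and no user parameters.
# Flash, shockwave, shell and floor-burst meshes expand over layered glow and darkness sprites, fading from hot yellow-orange to dark purple.
# A flickering light, flipbook smoke and purple-blue sparks add detail.
# GPU emitters handle gravity-driven debris and hundreds of curl-noise specks.
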NiagaraSystem /CorruptionGameplay/Gameplay/Items/Weapons/Melee/Scythe/Effects/NS_PowerUp_Scythe_Impact
dependencies {
 usedDataInterfaces [
  &NiagaraDataInterfaceMeshRendererInfo
  &NiagaraDataInterfaceVectorCurve
  &NiagaraDataInterfaceCurve
  &NiagaraDataInterfaceColorCurve
  &NiagaraDataInterfaceVector2DCurve
  &NiagaraDataInterfaceCurlNoise
 ]
 usedDynamicInputs [&Vector2DFromCurve]
 usedModules [&SystemState]
}
emitters {
 Darkness {
  emitterUpdateScript [
   &EmitterState(Life Cycle Mode=::NewEnumerator1, Loop Behavior=::NewEnumerator1, Loop Duration=1, UseLoopDelay=true)
   &SpawnBurst_Instantaneous(Use Spawn Probability=false, UseLoopCountLimit=false)
  ]
  particleSpawnScript [
   &InitializeParticle(Lifetime=0.4, Sprite Size Mode=::NewEnumerator3, Uniform Sprite Size=1, UsePositionOffset=false, Write Material Random=true)
   /SetVariables(Particles.DynamicMaterialParamGlow:Vector4f=&MakeVector4(X=1000, Y=1, Z=1, W=1)) {isSetParameters=true}
  ]
  particleUpdateScript [
   &ParticleState()
   &FN_Sprite_Attributes(
    Camera Offset=150
    Color=&MultiplyLinearColor(
     Linear Color=&ColorFromCurve(ColorCurve=DataInterface() {propertyValues {
        AlphaCurve {keys [<at 0 1 RCIM_Cubic>]}
        BlueCurve {keys [<at 0 0.349>, {interpMode=RCIM_Cubic, leaveTangent=-0.2232, time=0.426, value=0.0762}, <at 0.8297 0.15>]}
        GreenCurve {keys [<at 0 0.9586>, {interpMode=RCIM_Cubic, leaveTangent=-1.0752, time=0.426, value=0.2423}, <at 0.8297 0>]}
        RedCurve {keys [<at 0 1>, {interpMode=RCIM_Cubic, leaveTangent=-1.0197, time=0.426, value=0.6172}, <at 0.8297 0.0908>]}
       }})
     Scale Factor=&MakeFloat4FromVectorAndFloat(
      "Vector (RGB)"=<1 1 1>
      "Float (Alpha)"=&FloatFromCurve(
       FloatCurve=DataInterface() {propertyValues {Curve {keys [
           {arriveTangent=9.0725, interpMode=RCIM_Cubic, leaveTangent=9.0725, tangentMode=RCTM_User, time=0, value=0}
           {arriveTangent=-2.5249, interpMode=RCIM_Cubic, leaveTangent=-2.5249, tangentMode=RCTM_User, time=0.1, value=1}
           {arriveTangent=-0.0405, interpMode=RCIM_Cubic, leaveTangent=-0.0405, tangentMode=RCTM_User, time=1, value=0}
          ]}}}
       Scale Curve=0.5
      )
     )
    )
    Create Camera Offset Attribute=true
    Create Color Attribute=true
    Create SpriteSize Attribute=true
    SpriteSize=&Vector2DFromFloat(Value=&FloatFromCurve(
      FloatCurve=DataInterface() {propertyValues {Curve {keys [
          {arriveTangent=9.0725, interpMode=RCIM_Cubic, time=0, value=0.7}
          {arriveTangent=-0.2, interpMode=RCIM_Cubic, leaveTangent=-0.2, time=0.1408, value=0.9847}
          {arriveTangent=-0.0405, interpMode=RCIM_Cubic, leaveTangent=-0.0405, tangentMode=RCTM_User, time=1, value=0.5}
         ]}}}
      Scale Curve=800
     ))
    Write to Module Namespace=true
   ) {module=FN_Sprite_Attributes001}
   &FN_Sprite_Attributes(
    Camera Offset=150
    Color=&MultiplyLinearColor(
     Linear Color=<rgba 0.0171 0.0059 0.03 1>
     Scale Factor=&MakeFloat4FromVectorAndFloat(
      "Vector (RGB)"=<1 1 1>
      "Float (Alpha)"=&FloatFromCurve(FloatCurve=DataInterface() {propertyValues {Curve {keys [
           {arriveTangent=9.0725, interpMode=RCIM_Cubic, leaveTangent=9.0725, tangentMode=RCTM_User, time=0, value=0}
           {arriveTangent=-2.5249, interpMode=RCIM_Cubic, leaveTangent=-2.5249, tangentMode=RCTM_User, time=0.1, value=0.9847}
           {arriveTangent=-0.0405, interpMode=RCIM_Cubic, leaveTangent=-0.0405, tangentMode=RCTM_User, time=1, value=0}
          ]}}})
     )
    )
    Create Camera Offset Attribute=true
    Create Color Attribute=true
    Create SpriteSize Attribute=true
    SpriteSize=&Vector2DFromFloat(Value=&FloatFromCurve(
      FloatCurve=DataInterface() {propertyValues {Curve {keys [
          {arriveTangent=9.0725, interpMode=RCIM_Cubic, time=0, value=0.7}
          {arriveTangent=-0.2, interpMode=RCIM_Cubic, leaveTangent=-0.2, time=0.1408, value=0.9847}
          {arriveTangent=-0.0405, interpMode=RCIM_Cubic, leaveTangent=-0.0405, tangentMode=RCTM_User, time=1, value=0.5}
         ]}}}
      Scale Curve=700
     ))
   )
  ]
  rendererClasses [&NiagaraSpriteRendererProperties]
  renderers [
   {rendererClass=&NiagaraSpriteRendererProperties}
   {rendererClass=&NiagaraSpriteRendererProperties, rendererIndex=1}
  ]
 }
 Flash {
  emitterUpdateScript [
   &EmitterState(Life Cycle Mode=::NewEnumerator1, Loop Behavior=::NewEnumerator1, Loop Duration=1, UseLoopDelay=true)
   &SpawnBurst_Instantaneous(Spawn Count=3, Use Spawn Probability=false, UseLoopCountLimit=false)
  ]
  particleSpawnScript [
   &InitializeParticle(
    Lifetime Max=0.15
    Lifetime Min=0.12
    Lifetime Mode=::NewEnumerator1
    Mesh Scale Mode=::NewEnumerator1
    Mesh Uniform Scale Max=0.7
    Mesh Uniform Scale Min=0.5
    UsePositionOffset=false
    Write Material Random=true
   )
   /SetVariables(Particles.MeshOrientation=&MakeQuatFromAxisAngle(
     Axis=&RandomRangeVector(Minimum=<0 0 0>, Maximum=<1 1 1>, Evaluation Type=::NewEnumerator1)
     AngleInDegrees=90
    )) {isSetParameters=true}
  ]
  particleUpdateScript [
   &ParticleState()
   &ScaleColor(
    Scale Alpha=&FloatFromCurve(FloatCurve=DataInterface() {propertyValues {Curve {keys [<at 0.25 1 RCIM_Cubic>,<at 1 0>]}}})
    Scale RGB=&Multiply_Vector(
     A=<5 5 5>
     B=&MakeVectorFromLinearColorRGB(Color=&ColorFromCurve(ColorCurve=DataInterface() {propertyValues {
         AlphaCurve {keys [<at 0 1 RCIM_Cubic>]}
         BlueCurve {keys [<at 0 0.349>, {interpMode=RCIM_Cubic, leaveTangent=-0.2232, time=0.426, value=0.0762}, <at 0.8297 0.15>]}
         GreenCurve {keys [<at 0 0.9586>, {interpMode=RCIM_Cubic, leaveTangent=-1.0752, time=0.426, value=0.2423}, <at 0.8297 0>]}
         RedCurve {keys [<at 0 1>, {interpMode=RCIM_Cubic, leaveTangent=-1.0197, time=0.426, value=0.6172}, <at 0.8297 0.0908>]}
        }}))
    )
    ScaleA=true
    ScaleRGB=true
   )
   &ScaleMeshSize(Scale Factor=&VectorFromFloat(Value=&FloatFromCurve(FloatCurve=DataInterface() {propertyValues {Curve {keys [
          {arriveTangent=3.0939, interpMode=RCIM_Cubic, leaveTangent=3.0939, tangentMode=RCTM_User, time=0, value=0}
          {arriveTangent=0.1314, interpMode=RCIM_Cubic, leaveTangent=0.1314, tangentMode=RCTM_User, time=1, value=2}
         ]}}})))
  ]
  rendererClasses [&NiagaraMeshRendererProperties]
  renderers [{rendererClass=&NiagaraMeshRendererProperties, rendererIndex=0}]
 }
 FloorBurst {
  emitterUpdateScript [
   &EmitterState(Life Cycle Mode=::NewEnumerator1, Loop Behavior=::NewEnumerator1, Loop Duration=1, UseLoopDelay=true)
   &SpawnBurst_Instantaneous(Use Spawn Probability=false, UseLoopCountLimit=false)
  ]
  particleSpawnScript [
   &InitializeParticle(Lifetime=0.4, Mesh Scale=<0.35 1.5 1.5>, Mesh Scale Mode=::NewEnumerator4, UsePositionOffset=false, Write Material Random=true)
   &InitialMeshOrientation(Orientation Vector=<-1 0 0>, Use Orientation Vector=true, Use Rotation Vector=false)
  ]
  particleUpdateScript [
   &ParticleState()
   &ScaleMeshSize(Scale Factor=&VectorFromCurve(VectorCurve=DataInterface() {propertyValues {
       XCurve {keys [{arriveTangent=1.6532, interpMode=RCIM_Cubic, time=0, value=0.3}, <at 1 1>]}
       YCurve {keys [{arriveTangent=1.2873, interpMode=RCIM_Cubic, time=0, value=0.1}, <at 1 1>]}
       ZCurve {keys [{arriveTangent=1.2847, interpMode=RCIM_Cubic, time=0, value=0.1}, <at 1 1>]}
      }}))
   &ScaleColor(
    Scale Alpha=&FloatFromCurve(FloatCurve=DataInterface() {propertyValues {Curve {keys [
         {arriveTangent=13.7695, interpMode=RCIM_Cubic, leaveTangent=13.7695, tangentMode=RCTM_User, time=0, value=0}
         <at 0.1927 1 RCIM_Cubic>
         <at 1 0>
        ]}}})
    Scale RGB=&Multiply_Vector(
     A=<5 5 5>
     B=&MakeVectorFromLinearColorRGB(Color=&ColorFromCurve(ColorCurve=DataInterface() {propertyValues {
         AlphaCurve {keys [<at 0 1 RCIM_Cubic>]}
         BlueCurve {keys [<at 0 0.349>, {interpMode=RCIM_Cubic, leaveTangent=-0.2232, time=0.426, value=0.0762}, <at 0.8297 0.15>]}
         GreenCurve {keys [<at 0 0.9586>, {interpMode=RCIM_Cubic, leaveTangent=-1.0752, time=0.426, value=0.2423}, <at 0.8297 0>]}
         RedCurve {keys [<at 0 1>, {interpMode=RCIM_Cubic, leaveTangent=-1.0197, time=0.426, value=0.6172}, <at 0.8297 0.0908>]}
        }}))
    )
    ScaleA=true
    ScaleRGB=true
   )
   &CameraOffset(Camera Offset Amount=80)
   &DynamicMaterialParameters(
    Index 0 Param 1=&FloatFromCurve(FloatCurve=DataInterface() {propertyValues {Curve {keys [{arriveTangent=3.5549, interpMode=RCIM_Cubic, time=0, value=0}, <at 1 2.5>]}}})
    Index 0 Param 2=0.7
    Index 0 Param 3=1
    Index 0 Param 4=1
    Module.Param0WriteEnabled=true
    Module.Param1WriteEnabled=true
    Module.Param2WriteEnabled=true
    Module.Param3WriteEnabled=true
   )
  ]
  rendererClasses [&NiagaraMeshRendererProperties]
  renderers [{rendererClass=&NiagaraMeshRendererProperties, rendererIndex=0}]
 }
 Light {
  emitterUpdateScript [
   &EmitterState(Life Cycle Mode=::NewEnumerator1, Loop Behavior=::NewEnumerator1, Loop Duration=1, UseLoopDelay=true)
   &SpawnBurst_Instantaneous(Use Spawn Probability=false, UseLoopCountLimit=false)
  ]
  particleSpawnScript [
   &InitializeParticle(Lifetime=0.64, Sprite Size Mode=::NewEnumerator3, Uniform Sprite Size=300, UsePositionOffset=false, Write Material Random=true)
   &Color(Color=&MakeLinearColorFromVectorAndFloat("Vector (RGB)"=<1 0 0.7431>, "Float (Alpha)"=&ClampFloat(Float=1)))
  ]
  particleUpdateScript [
   /SetVariables(Particles.LightExponent=5) {isSetParameters=true}
   /SetVariables(Particles.LightColor:Vector3f=<1 1 1>) {isSetParameters=true}
   /SetVariables(Particles.LightRadius=1) {isSetParameters=true}
   &ParticleState()
   &ScaleColor(
    Scale Alpha=&FloatFromCurve(FloatCurve=DataInterface() {propertyValues {Curve {keys [
         <at 0.2 1 RCIM_Cubic>
         {arriveTangent=-1.5549, interpMode=RCIM_Cubic, leaveTangent=-1.5549, time=0.5559, value=0.0539}
         {arriveTangent=-0.6873, interpMode=RCIM_Cubic, leaveTangent=-0.6873, time=0.5593, value=0.4413}
         {arriveTangent=-2.2445, interpMode=RCIM_Cubic, leaveTangent=-2.2445, time=0.5889, value=0.0312}
         {arriveTangent=-0.5618, interpMode=RCIM_Cubic, leaveTangent=-0.5618, time=0.5925, value=0.3669}
         {arriveTangent=-12.0689, interpMode=RCIM_Cubic, leaveTangent=-12.0689, time=0.6119, value=0.0183}
         {arriveTangent=0.4595, interpMode=RCIM_Cubic, leaveTangent=0.4595, time=0.6125, value=0.125}
         {arriveTangent=2.2462, interpMode=RCIM_Cubic, leaveTangent=2.2462, time=0.6262, value=0.0249}
         {arriveTangent=-1.0472, interpMode=RCIM_Cubic, leaveTangent=-1.0472, time=0.632, value=0.1688}
         {arriveTangent=-4.6153, interpMode=RCIM_Cubic, leaveTangent=-4.6153, time=0.65, value=0}
         {arriveTangent=3.6989, interpMode=RCIM_Cubic, leaveTangent=3.6989, time=0.6686, value=0}
         <at 0.6748 0.0917 RCIM_Cubic>
         {arriveTangent=-3.4671, interpMode=RCIM_Cubic, leaveTangent=-3.4671, time=0.685, value=0}
         {arriveTangent=4.152, interpMode=RCIM_Cubic, leaveTangent=4.152, time=0.6994, value=0.0063}
         {arriveTangent=-0.3173, interpMode=RCIM_Cubic, leaveTangent=-0.3173, time=0.703, value=0.0751}
         {arriveTangent=-1.3079, interpMode=RCIM_Cubic, leaveTangent=-1.3079, time=0.7194, value=0}
         {arriveTangent=2.0927, interpMode=RCIM_Cubic, leaveTangent=2.0927, time=0.7605, value=0}
         <at 0.7608 0.0866 RCIM_Cubic>
         {arriveTangent=-0.9264, interpMode=RCIM_Cubic, leaveTangent=-0.9264, time=0.7882, value=0}
         {arriveTangent=1.001, interpMode=RCIM_Cubic, leaveTangent=1.001, time=0.8543, value=0}
         {arriveTangent=0.9418, interpMode=RCIM_Cubic, leaveTangent=0.9418, time=0.8568, value=0.0687}
         {arriveTangent=-0.7198, interpMode=RCIM_Cubic, leaveTangent=-0.7198, time=0.8801, value=0.0243}
         {arriveTangent=0.4725, interpMode=RCIM_Cubic, leaveTangent=0.4725, time=0.9522, value=0}
         {arriveTangent=1.2274, interpMode=RCIM_Cubic, leaveTangent=1.2274, time=0.9561, value=0.0602}
         {arriveTangent=-1.3699, interpMode=RCIM_Cubic, leaveTangent=-1.3699, time=0.969, value=0.0207}
         <at 1 0>
        ]}}})
    Scale RGB=&VectorFromCurve(VectorCurve=DataInterface() {propertyValues {
       XCurve {keys [<at 0 1 RCIM_Cubic>,<at 1 1>]}
       YCurve {keys [<at 0 1 RCIM_Cubic>,<at 1 0.3719>]}
       ZCurve {keys [<at 0 1 RCIM_Cubic>,<at 1 0.4317>]}
      }})
    ScaleA=true
    ScaleRGB=true
   )
  ]
  rendererClasses [&NiagaraLightRendererProperties]
  renderers [{rendererClass=&NiagaraLightRendererProperties, rendererIndex=0}]
 }
 ShellBurst {
  emitterUpdateScript [
   &EmitterState(Life Cycle Mode=::NewEnumerator1, Loop Behavior=::NewEnumerator1, Loop Duration=1, UseLoopDelay=true)
   &SpawnBurst_Instantaneous(Spawn Time=0.02, Use Spawn Probability=false, UseLoopCountLimit=false)
  ]
  particleSpawnScript [
   &InitializeParticle(Lifetime=0.3, Mesh Scale Mode=::NewEnumerator3, Mesh Uniform Scale=1.25, UsePositionOffset=false, Write Material Random=true)
   &InitialMeshOrientation(Orientation Vector=<-1 0 0>, Use Orientation Vector=true, Use Rotation Vector=false)
   &DynamicMaterialParameters(
    Index 0 Param 1=0.5
    Index 0 Param 2=1
    Index 0 Param 3=1
    Index 0 Param 4=1
    Module.Param0WriteEnabled=true
    Module.Param1WriteEnabled=true
    Module.Param2WriteEnabled=true
    Module.Param3WriteEnabled=true
   )
  ]
  particleUpdateScript [
   &ParticleState()
   &ScaleMeshSize(Scale Factor=&VectorFromCurve(VectorCurve=DataInterface() {propertyValues {
       XCurve {keys [
         {arriveTangent=1.6643, interpMode=RCIM_Cubic, leaveTangent=1.6643, tangentMode=RCTM_User, time=0, value=0}
         {arriveTangent=0.3199, interpMode=RCIM_Cubic, leaveTangent=0.3199, tangentMode=RCTM_User, time=1, value=1}
        ]}
       YCurve {keys [{arriveTangent=3.3122, interpMode=RCIM_Cubic, leaveTangent=3.3122, tangentMode=RCTM_User, time=0, value=0}, <at 1 1 RCIM_Cubic>]}
       ZCurve {keys [{arriveTangent=3.3122, interpMode=RCIM_Cubic, leaveTangent=3.3122, tangentMode=RCTM_User, time=0, value=0}, <at 1 1 RCIM_Cubic>]}
      }}))
   &ScaleColor(
    Scale Alpha=&FloatFromCurve(FloatCurve=DataInterface() {propertyValues {Curve {keys [<at 0 1 RCIM_Cubic>, {arriveTangent=-1, interpMode=RCIM_Cubic, leaveTangent=-1, time=0.1927, value=0.7374}, <at 1 0>]}}})
    Scale RGB=&Multiply_Vector(
     A=<5 5 5>
     B=&MakeVectorFromLinearColorRGB(Color=&ColorFromCurve(ColorCurve=DataInterface() {propertyValues {
         AlphaCurve {keys [<at 0 1 RCIM_Cubic>]}
         BlueCurve {keys [<at 0 0.349>, {interpMode=RCIM_Cubic, leaveTangent=-0.2232, time=0.426, value=0.0762}, <at 0.8049 0.15>]}
         GreenCurve {keys [<at 0 0.9586>, {interpMode=RCIM_Cubic, leaveTangent=-1.0752, time=0.426, value=0.2423}, <at 0.8049 0>]}
         RedCurve {keys [<at 0 1>, {interpMode=RCIM_Cubic, leaveTangent=-1.0197, time=0.426, value=0.6172}, <at 0.8049 0.0908>]}
        }}))
    )
    ScaleA=true
    ScaleRGB=true
   )
   &CameraOffset(Camera Offset Amount=80)
  ]
  rendererClasses [&NiagaraMeshRendererProperties]
  renderers [{rendererClass=&NiagaraMeshRendererProperties, rendererIndex=0}]
 }
 ShockWave {
  emitterUpdateScript [
   &EmitterState(Life Cycle Mode=::NewEnumerator1, Loop Behavior=::NewEnumerator1, Loop Duration=1, UseLoopDelay=true)
   &SpawnBurst_Instantaneous(Use Spawn Probability=false, UseLoopCountLimit=false)
  ]
  particleSpawnScript [
   &InitializeParticle(Lifetime=0.18, Mesh Scale Mode=::NewEnumerator3, Mesh Uniform Scale=3, UsePositionOffset=false, Write Material Random=true)
   /SetVariables(Particles.MeshOrientation=&MakeQuatFromAxisAngle(Axis=<0 1 0>, AngleInDegrees=90, NormalizeAxis=true)) {isSetParameters=true}
  ]
  particleUpdateScript [
   &ParticleState()
   &ScaleMeshSize(Scale Factor=&VectorFromFloat(Value=&FloatFromCurve(FloatCurve=DataInterface() {propertyValues {Curve {keys [
          {arriveTangent=5.6511, interpMode=RCIM_Cubic, leaveTangent=5.6511, tangentMode=RCTM_User, time=0, value=0}
          {arriveTangent=1, interpMode=RCIM_Cubic, leaveTangent=1, time=0.2179, value=0.6941}
          {interpMode=RCIM_Cubic, leaveTangent=0.2231, time=1, value=1}
         ]}}})))
   &DynamicMaterialParameters(
    Index 0 Param 1=&FloatFromCurve(FloatCurve=DataInterface() {propertyValues {Curve {keys [{arriveTangent=1.074, interpMode=RCIM_Cubic, time=0, value=0.5}, <at 1 1>]}}})
    Index 0 Param 2=&FloatFromCurve(FloatCurve=DataInterface() {propertyValues {Curve {keys [<at 0 1.25 RCIM_Cubic>, {arriveTangent=-3.093, leaveTangent=-3.093, time=1, value=0}]}}})
    Index 0 Param 3=&FloatFromCurve(FloatCurve=DataInterface() {propertyValues {Curve {keys [{arriveTangent=0.6796, interpMode=RCIM_Cubic, time=0, value=-0.5}, <at 1 0>]}}})
    Index 0 Param 4=&FloatFromCurve(FloatCurve=DataInterface() {propertyValues {Curve {keys [{arriveTangent=0.2861, interpMode=RCIM_Cubic, time=0, value=0}, <at 1 0.12>]}}})
    Module.Param0WriteEnabled=true
    Module.Param1WriteEnabled=true
    Module.Param2WriteEnabled=true
    Module.Param3WriteEnabled=true
   )
   &ScaleColor(
    Scale Alpha=&FloatFromCurve(FloatCurve=DataInterface() {propertyValues {Curve {keys [<at 0 1 RCIM_Cubic>, {arriveTangent=-1, interpMode=RCIM_Cubic, leaveTangent=-1, time=0.5708, value=0.7468}, <at 1 0>]}}})
    Scale RGB=&Multiply_Vector(
     A=<5 5 5>
     B=&MakeVectorFromLinearColorRGB(Color=&ColorFromCurve(ColorCurve=DataInterface() {propertyValues {
         AlphaCurve {keys [<at 0 1 RCIM_Cubic>]}
         BlueCurve {keys [<at 0 0.349>, {interpMode=RCIM_Cubic, leaveTangent=-0.2232, time=0.426, value=0.0762}, <at 0.8297 0.15>]}
         GreenCurve {keys [<at 0 0.9586>, {interpMode=RCIM_Cubic, leaveTangent=-1.0752, time=0.426, value=0.2423}, <at 0.8297 0>]}
         RedCurve {keys [<at 0 1>, {interpMode=RCIM_Cubic, leaveTangent=-1.0197, time=0.426, value=0.6172}, <at 0.8297 0.0908>]}
        }}))
    )
    ScaleA=true
    ScaleRGB=true
   )
   /SetVariables(Particles.ColorLow:LinearColor=&MultiplyLinearColor(Linear Color=$Particles.Color, Scale Factor=<0.5 0.5 0.5 0.25>)) {isSetParameters=true}
  ]
  rendererClasses [&NiagaraMeshRendererProperties]
  renderers [
   {rendererClass=&NiagaraMeshRendererProperties, rendererIndex=0}
   {rendererClass=&NiagaraMeshRendererProperties, rendererIndex=1}
  ]
 }
 Smoke {
  emitterUpdateScript [
   &EmitterState(Life Cycle Mode=::NewEnumerator1, Loop Behavior=::NewEnumerator1, Loop Delay=0.06, Loop Duration=1, UseLoopDelay=true)
   &SpawnBurst_Instantaneous(Spawn Count=35, Use Spawn Probability=false, UseLoopCountLimit=false)
  ]
  particleSpawnScript [
   &InitializeParticle(
    Color Minimum=<rgba 0.25 0.25 0.25 1>
    Color Mode=::NewEnumerator2
    Lifetime Max=0.75
    Lifetime Min=0.3
    Lifetime Mode=::NewEnumerator1
    Sprite Rotation Mode=::NewEnumerator2
    Sprite Size Mode=::NewEnumerator1
    Uniform Sprite Size Max=160
    Uniform Sprite Size Min=100
    UsePositionOffset=false
    Write Material Random=true
   ) {module=InitializeParticle001}
   &CylinderLocation(
    Cylinder Height=50
    Cylinder Radius=5
    MaskSpawnGroup=false
    Override Seed=false
    Rotation Mode:ENiagara_RotationMode=::NewEnumerator1
    Surface Only=true
    Use Endcaps In Surface Only Mode=false
    UseOwnerScale=true
   )
   &AddVelocity(
    Coordinate Space=::Simulation
    Velocity=&Multiply_VectorByFloat(Vector=$Particles.CylinderLocation.CylinderNormal, Float=1200)
   )
   &AddVelocity(
    Coordinate Space=::World
    Velocity=&RandomRangeVector(Minimum=<0 0 150>, Maximum=<0 0 300>, Evaluation Type=::NewEnumerator1)
   ) {module=AddVelocity001}
   /SetVariables(Particles.RandomFrame=&RandomRangeInt(Minimum=0, Maximum=63)) {isSetParameters=true}
  ]
  particleUpdateScript [
   &ParticleState()
   &Drag(Drag=&RandomRangeFloat(Minimum=5, Maximum=7), Use Linear Drag=true, Use Rotational Drag=false)
   &SolveForcesAndVelocity(Clamp Velocity=false, Limit Acceleration=false)
   &ScaleColor(
    Scale Alpha=&FloatFromCurve(
     FloatCurve=DataInterface() {propertyValues {Curve {keys [
         {arriveTangent=15.7334, interpMode=RCIM_Cubic, leaveTangent=15.7334, tangentMode=RCTM_User, time=0, value=0}
         {arriveTangent=0.3312, interpMode=RCIM_Cubic, leaveTangent=-0.2993, tangentMode=RCTM_Break, time=0.1, value=1}
         {interpMode=RCIM_Cubic, leaveTangent=-0.6165, time=1, value=0}
        ]}}}
     Scale Curve=&RandomRangeFloat(Minimum=1.25, Maximum=2)
    )
    Scale RGB=&MakeVectorFromLinearColorRGB(Color=&ColorFromCurve(ColorCurve=DataInterface() {propertyValues {
        AlphaCurve {keys [<at 0 1 RCIM_Cubic>]}
        BlueCurve {keys [
          <at 0 0.349>
          {arriveTangent=-0.8941, interpMode=RCIM_Cubic, leaveTangent=-0.8941, time=0.1044, value=0.0762}
          <at 0.3773 0.0116>
          <at 0.7682 0.15>
         ]}
        GreenCurve {keys [
          <at 0 0.9586>
          {arriveTangent=-2.4395, interpMode=RCIM_Cubic, leaveTangent=-2.4395, time=0.1044, value=0.2423}
          <at 0.3773 0.0382>
          <at 0.7682 0>
         ]}
        RedCurve {keys [
          <at 0 1>
          {arriveTangent=-2.3856, interpMode=RCIM_Cubic, leaveTangent=-2.3856, time=0.1044, value=0.6172}
          <at 0.3773 0.0999>
          <at 0.7682 0.0908>
         ]}
       }}))
    ScaleA=true
    ScaleRGB=true
   ) {module=ScaleColor001}
   &ScaleSpriteSize(Scale Factor=&Vector2DFromFloat(Value=&FloatFromCurve(FloatCurve=DataInterface() {propertyValues {Curve {keys [
          {arriveTangent=2.8432, interpMode=RCIM_Cubic, leaveTangent=2.8432, tangentMode=RCTM_User, time=0.0023, value=0.2}
          {interpMode=RCIM_Cubic, leaveTangent=0.0897, time=1, value=1}
         ]}}})))
   &FN_SubUVAnimation(
    End Frame=&Add_Integer(A=$Particles.RandomFrame, B=63)
    Start Frame=$Particles.RandomFrame
    SubUV Animation Mode:ENiagara_SubUVLookupMode=::NewEnumerator1
    SubUV Curve=DataInterface() {propertyValues {Curve {keys [{arriveTangent=1.9163, interpMode=RCIM_Cubic, leaveTangent=1.9163, tangentMode=RCTM_User, time=0, value=0}, <at 1 1 RCIM_Cubic>]}}}
    SubUV Loop Count=0.5
   )
  ]
  rendererClasses [&NiagaraSpriteRendererProperties]
  renderers [{rendererClass=&NiagaraSpriteRendererProperties}]
 }
 Spark {
  emitterUpdateScript [
   &EmitterState(Life Cycle Mode=::NewEnumerator1, Loop Behavior=::NewEnumerator1, Loop Delay=0.03, Loop Duration=1, UseLoopDelay=true)
   &SpawnBurst_Instantaneous(Spawn Count=25, Use Spawn Probability=false)
  ]
  particleSpawnScript [
   &InitializeParticle(
    Lifetime Max=0.4
    Lifetime Min=0.15
    Lifetime Mode=::NewEnumerator1
    Sprite Size Max=<4 6>
    Sprite Size Min=<2 3>
    Sprite Size Mode=::NewEnumerator2
    UsePositionOffset=false
    Write Material Random=true
   )
   &CylinderLocation(
    Cylinder Height=0.1
    Cylinder Radius=30
    MaskSpawnGroup=false
    Offset=&RandomRangeVector(Minimum=<-10 0 0>, Maximum=<10 0 0>, Evaluation Type=::NewEnumerator1)
    Offset Coordinate Space:ENiagaraCoordinateSpace=::Local
    Offset Mode:ENiagara_OffsetMode=::NewEnumerator0
    Override Seed=false
    Rotation Angle Type:ENiagara_AngleInput=::NewEnumerator0
    Rotation Mode:ENiagara_RotationMode=::NewEnumerator1
    UseOwnerScale=true
   )
   &AddVelocity(
    Coordinate Space=::Simulation
    Velocity=&Multiply_VectorByFloat(
     Vector=$Particles.CylinderLocation.CylinderNormal
     Float=&RandomRangeFloat(Minimum=300, Maximum=900, Evaluation Type=::NewEnumerator1)
    )
   )
   &AddVelocity(
    Coordinate Space=::World
    Velocity=&RandomRangeVector(Minimum=<0 0 100>, Maximum=<0 0 600>, Evaluation Type=::NewEnumerator1)
   ) {module=AddVelocity001}
  ]
  particleUpdateScript [
   &ParticleState()
   &ScaleColor(
    Scale Alpha=&FloatFromCurve(FloatCurve=DataInterface() {propertyValues {Curve {keys [<at 0.66 1 RCIM_Cubic>,<at 1 0 RCIM_Cubic>]}}}, Scale Curve=0.5)
    Scale RGB=&VectorFromFloat(Value=&FloatFromCurve(
      FloatCurve=DataInterface() {propertyValues {Curve {keys [<at 0.5 4 RCIM_Cubic>,<at 1 2 RCIM_Cubic>]}}}
      Scale Curve=&FNTimeOfDayScalar(Night Scale=1, Morning Scale=1, Evening Scale=1)
     ))
    ScaleA=true
    ScaleRGB=true
   ) {module=ScaleColor002}
   &ScaleColor(
    Linear Color Curve=DataInterface() {propertyValues {
      AlphaCurve {keys [<at 0 1 RCIM_Cubic>]}
      BlueCurve {keys [<at 0 1>, {interpMode=RCIM_Cubic, leaveTangent=-0.85, time=0.7675, value=1}]}
      GreenCurve {keys [<at 0 0>,<at 0.7675 0.125 RCIM_Cubic>]}
      RedCurve {keys [<at 0 0.7379>, {interpMode=RCIM_Cubic, leaveTangent=-0.6471, time=0.7675, value=0.2708}]}
     }}
    Scale Mode=::NewEnumerator2
   ) {module=ScaleColor003}
   &GravityForce(Gravity=&RandomRangeVector(Minimum=<0 0 0>, Maximum=<0 0 -100>))
   &CurlNoiseForce(
    Noise Frequency=20
    Noise Strength=&Multiply_Float(
     A=&RandomRangeFloat(Minimum=1000, Maximum=1500)
     B=&FloatFromCurve(FloatCurve=DataInterface() {propertyValues {Curve {keys [<at 0 0>,<at 0.2 1>]}}})
    )
    Pan Noise=true
   )
   &Drag(Drag=&RandomRangeFloat(Minimum=4, Maximum=8), Use Linear Drag=true, Use Rotational Drag=false)
   &SolveForcesAndVelocity(Clamp Velocity=false, Limit Acceleration=false)
   &ScaleSpriteSizeBySpeed(Max Scale Factor=<0.25 5>, Min Scale Factor=<1 0.5>, Velocity Threshold=750)
   &ScaleSpriteSize(Scale Factor=&MakeVector2D(
     X=0.66
     Y=&FloatFromCurve(FloatCurve=DataInterface() {propertyValues {Curve {keys [<at 0 0.5 RCIM_Cubic>,<at 0.1 1.5 RCIM_Cubic>,<at 1 0.5>]}}})
    ))
  ]
  rendererClasses [&NiagaraSpriteRendererProperties]
  renderers [{rendererClass=&NiagaraSpriteRendererProperties}]
 }
 Sparks_Initial {
  emitterUpdateScript [
   &EmitterState(Life Cycle Mode=::NewEnumerator1, Loop Behavior=::NewEnumerator1, Loop Duration=1, UseLoopDelay=true)
   &SpawnBurst_Instantaneous(
    Spawn Count=&RandomRangeInt(Minimum=18, Maximum=25)
    Use Spawn Probability=false
    UseLoopCountLimit=false
   )
  ]
  particleSpawnScript [
   &InitializeParticle(
    Lifetime Max=0.14
    Lifetime Min=0.08
    Lifetime Mode=::NewEnumerator1
    Sprite Size Max=<6 6>
    Sprite Size Min=<4 1>
    Sprite Size Mode=::NewEnumerator2
    UsePositionOffset=false
    Write Material Random=true
   )
   &SphereLocation(
    Hemisphere X=true
    MaskSpawnGroup=false
    Override Seed=false
    Sphere Radius=50
    Surface Only=false
    UseNonUniformScale=false
    UseOffset=false
    UseSphereOrientation=false
   )
   &AddVelocity(
    Coordinate Space=::Simulation
    Velocity=&Multiply_VectorByFloat(Vector=$Output.SphereLocation.SphereVector, Float=&RandomRangeFloat(Minimum=100, Maximum=120))
   )
   &AddVelocity(
    Coordinate Space=::World
    Velocity=&RandomRangeVector(Minimum=<0 0 400>, Maximum=<0 0 800>, Evaluation Type=::NewEnumerator1)
   ) {module=AddVelocity001}
  ]
  particleUpdateScript [
   &ParticleState()
   &ScaleColor(
    Linear Color Curve=DataInterface() {propertyValues {
      AlphaCurve {keys [<at 0 1 RCIM_Cubic>]}
      BlueCurve {keys [<at 0 1>, {interpMode=RCIM_Cubic, leaveTangent=-0.85, time=0.7675, value=1}]}
      GreenCurve {keys [<at 0 0>,<at 0.7675 0.125 RCIM_Cubic>]}
      RedCurve {keys [<at 0 0.7379>, {interpMode=RCIM_Cubic, leaveTangent=-0.6471, time=0.7675, value=0.2708}]}
     }}
    Scale Mode=::NewEnumerator2
   ) {module=ScaleColor001}
   &ScaleColor(
    Scale Alpha=&FloatFromCurve(FloatCurve=DataInterface() {propertyValues {Curve {keys [<at 0.66 1 RCIM_Cubic>,<at 1 0 RCIM_Cubic>]}}}, Scale Curve=0.5)
    Scale RGB=&VectorFromFloat(Value=&FloatFromCurve(
      FloatCurve=DataInterface() {propertyValues {Curve {keys [<at 0.5 2 RCIM_Cubic>,<at 1 0.5 RCIM_Cubic>]}}}
      Scale Curve=&FNTimeOfDayScalar(Night Scale=1, Morning Scale=1, Evening Scale=1)
     ))
    ScaleA=true
    ScaleRGB=true
   ) {module=ScaleColor002}
   &ScaleColor(
    Scale RGB=&Multiply_Vector(
     A=<5 5 5>
     B=&MakeVectorFromLinearColorRGB(Color=&ColorFromCurve(ColorCurve=DataInterface() {propertyValues {
         AlphaCurve {keys [<at 0 1 RCIM_Cubic>]}
         BlueCurve {keys [<at 0 0.2292>, {interpMode=RCIM_Cubic, leaveTangent=-0.3033, time=0.4327, value=0.0762}, <at 1 0.0061>]}
         GreenCurve {keys [<at 0 0.951>, {interpMode=RCIM_Cubic, leaveTangent=-1.2663, time=0.4327, value=0.2423}, <at 1 0.0197>]}
         RedCurve {keys [<at 0 1>, {interpMode=RCIM_Cubic, leaveTangent=-1.2917, time=0.4327, value=0.6172}, <at 1 0.05>]}
        }}))
    )
    ScaleA=true
    ScaleRGB=true
   ) {enabled=false}
   &Drag(Drag=10, Use Linear Drag=true, Use Rotational Drag=false)
   &SolveForcesAndVelocity(Clamp Velocity=false, Limit Acceleration=false)
   &ScaleSpriteSizeBySpeed(
    Max Scale Factor=<1 30>
    Min Scale Factor=<1 0.01>
    Source Velocity:Vector3f=&Multiply_Vector(A=<0 0.01 1>, B=$Particles.Velocity)
    Velocity Threshold=800
   )
   &ScaleSpriteSize(Uniform Curve Sprite Scale=DataInterface() {propertyValues {Curve {keys [{interpMode=RCIM_Cubic, tangentMode=RCTM_User, time=0, value=1}, {interpMode=RCIM_Cubic, tangentMode=RCTM_User, time=1, value=0}]}}})
  ]
  rendererClasses [&NiagaraSpriteRendererProperties]
  renderers [{rendererClass=&NiagaraSpriteRendererProperties}]
 }
 Spots {
  simTarget=GPUComputeSim
  emitterUpdateScript [
   &EmitterState(Life Cycle Mode=::NewEnumerator1, Loop Behavior=::NewEnumerator1, Loop Delay=0.13, Loop Duration=1, UseLoopDelay=true)
   &SpawnBurst_Instantaneous(Spawn Count=600, Use Spawn Probability=false)
  ]
  particleSpawnScript [
   &InitializeParticle(
    Color=<rgba 1 1 1 0.5>
    Lifetime Max=1.5
    Lifetime Min=0.33
    Lifetime Mode=::NewEnumerator1
    Sprite Size Mode=::NewEnumerator1
    Uniform Sprite Size Max=3
    Uniform Sprite Size Min=1
    UsePositionOffset=false
    Write Material Random=true
   )
   &SphereLocation(
    Hemisphere X=true
    MaskSpawnGroup=false
    Override Seed=false
    Sphere Radius=50
    Surface Only=false
    UseNonUniformScale=false
    UseOffset=false
    UseSphereOrientation=false
   )
   &AddVelocity(
    Coordinate Space=::Simulation
    Velocity=&Multiply_VectorByFloat(Vector=$Output.SphereLocation.SphereVector, Float=&RandomRangeFloat(Minimum=10, Maximum=15))
   )
   &AddVelocity(
    Coordinate Space=::World
    Velocity=&RandomRangeVector(Minimum=<0 0 400>, Maximum=<0 0 800>, Evaluation Type=::NewEnumerator1)
   ) {enabled=false, module=AddVelocity001}
  ]
  particleUpdateScript [
   &ParticleState()
   &ScaleColor(
    Scale Alpha=&FloatFromCurve(FloatCurve=DataInterface() {propertyValues {Curve {keys [
         {tangentMode=RCTM_User, time=0, value=1}
         {arriveTangent=-0.0405, interpMode=RCIM_Cubic, leaveTangent=-0.0405, tangentMode=RCTM_User, time=0.5, value=1}
         {interpMode=RCIM_Cubic, leaveTangent=-8.0089, time=1, value=0}
        ]}}})
    Scale RGB=&MakeCustomVectorFromBool(
     Bool=&RandomBool()
     True Vector=&Lerp_Vector(
      A=&MakeVectorFromLinearColorRGB(Color=<rgba 1 0.5158 0.5094 1>)
      B=&MakeVectorFromLinearColorRGB(Color=<rgba 0.2292 0.0262 0.066 1>)
      Alpha=&ClampFloat(Float=&Multiply_Float(A=$Particles.NormalizedAge, B=2.5))
     )
     False Vector=&Lerp_Vector(
      A=&MakeVectorFromLinearColorRGB(Color=<rgba 0.7379 0 1 1>)
      B=&MakeVectorFromLinearColorRGB(Color=<rgba 0.008 0 0.0685 1>)
      Alpha=&ClampFloat(Float=&Multiply_Float(A=$Particles.NormalizedAge, B=2.5))
     )
    )
    ScaleA=true
    ScaleRGB=true
   )
   &GravityForce(Gravity=&RandomRangeVector(Minimum=<0 0 0>, Maximum=<0 0 -100>))
   &CurlNoiseForce(
    Noise Frequency=25
    Noise Strength=&Multiply_Float(
     A=&RandomRangeFloat(Minimum=1000, Maximum=1500)
     B=&FloatFromCurve(FloatCurve=DataInterface() {propertyValues {Curve {keys [<at 0 0>,<at 0.2 1>]}}})
    )
    Pan Noise=true
   )
   &Drag(Drag=&RandomRangeFloat(Minimum=2, Maximum=6), Use Linear Drag=true, Use Rotational Drag=false)
   &SolveForcesAndVelocity(Clamp Velocity=false, Limit Acceleration=false)
   &ScaleSpriteSize(Scale Factor=&Vector2DFromFloat(Value=&FloatFromCurve(FloatCurve=DataInterface() {propertyValues {Curve {keys [
          {tangentMode=RCTM_User, time=0, value=1}
          {arriveTangent=-0.0405, interpMode=RCIM_Cubic, leaveTangent=-0.0405, tangentMode=RCTM_User, time=0.7009, value=1}
          {arriveTangent=-8.0089, interpMode=RCIM_Cubic, leaveTangent=-8.0089, tangentMode=RCTM_User, time=1, value=0}
         ]}}})))
  ]
  rendererClasses [&NiagaraSpriteRendererProperties]
  renderers [{rendererClass=&NiagaraSpriteRendererProperties}]
 }
 debrisGPU {
  simTarget=GPUComputeSim
  emitterUpdateScript [
   &EmitterState(Life Cycle Mode=::NewEnumerator1, Loop Behavior=::NewEnumerator1, Loop Duration=1, UseLoopDelay=true)
   &SpawnBurst_Instantaneous(
    Spawn Count=&RandomRangeInt(Minimum=15, Maximum=20)
    Spawn Time=0.03
    Use Spawn Probability=false
    UseLoopCountLimit=false
   )
  ]
  particleSpawnScript [
   &InitializeParticle(
    Color Maximum=<rgba 0.5 0.5 0.5 1>
    Color Minimum=<rgba 0.25 0.25 0.25 1>
    Color Mode=::NewEnumerator2
    Lifetime Mode=::NewEnumerator1
    Sprite Rotation Mode=::NewEnumerator2
    Sprite Size Mode=::NewEnumerator1
    Uniform Sprite Size Max=12
    UsePositionOffset=false
    Write Material Random=true
   ) {module=InitializeParticle001}
   &SphereLocation(
    MaskSpawnGroup=false
    Non Uniform Scale=<0.25 1 1>
    Override Seed=false
    Sphere Radius=50
    Surface Only=false
    UseNonUniformScale=true
    UseOffset=false
    UseSphereOrientation=false
   )
   &AddVelocity(
    Coordinate Space=::Simulation
    Velocity=&Multiply_VectorByFloat(Vector=$Output.SphereLocation.SphereVector, Float=&RandomRangeFloat(Minimum=5, Maximum=25))
   )
   &AddVelocity(
    Coordinate Space=::World
    Velocity=&RandomRangeVector(Minimum=<0 0 250>, Maximum=<0 0 600>, Evaluation Type=::NewEnumerator1)
   ) {module=AddVelocity001}
   &SubUVAnimation(End Frame=3, SubUV Animation Mode:ENiagara_SubUVLookupMode=::NewEnumerator2) {module=SubUVAnimation001}
  ]
  particleUpdateScript [
   &ParticleState()
   &AccelerationForce(Acceleration=<0 0 -980>)
   &SpriteRotationRate(Rotation Rate=&MakeCustomFloatFromBool(Bool=&RandomBool(), True Float=-1080, False Float=1080))
   &SolveForcesAndVelocity(Clamp Velocity=false, Limit Acceleration=false)
   &ScaleSpriteSize(Uniform Curve Sprite Scale=DataInterface() {propertyValues {Curve {keys [
        {interpMode=RCIM_Cubic, tangentMode=RCTM_User, time=0.33, value=1}
        {arriveTangent=-2.273, interpMode=RCIM_Cubic, leaveTangent=-2.273, tangentMode=RCTM_User, time=1, value=0}
       ]}}})
  ]
  rendererClasses [&NiagaraSpriteRendererProperties]
  renderers [{rendererClass=&NiagaraSpriteRendererProperties}]
 }
}
summary {
 systemName=NS_PowerUp_Scythe_Impact
 emitters [
  {emitterName=Light, rendererClasses [&NiagaraLightRendererProperties]}
  {emitterName=Darkness, rendererClasses [&NiagaraSpriteRendererProperties]}
  {emitterName=ShockWave, rendererClasses [&NiagaraMeshRendererProperties]}
  {emitterName=debrisGPU, simTarget=GPUComputeSim, rendererClasses [&NiagaraSpriteRendererProperties]}
  {emitterName=Flash, rendererClasses [&NiagaraMeshRendererProperties]}
  {emitterName=Sparks_Initial, rendererClasses [&NiagaraSpriteRendererProperties]}
  {emitterName=ShellBurst, rendererClasses [&NiagaraMeshRendererProperties]}
  {emitterName=FloorBurst, rendererClasses [&NiagaraMeshRendererProperties]}
  {emitterName=Spark, rendererClasses [&NiagaraSpriteRendererProperties]}
  {emitterName=Spots, simTarget=GPUComputeSim, rendererClasses [&NiagaraSpriteRendererProperties]}
  {emitterName=Smoke, rendererClasses [&NiagaraSpriteRendererProperties]}
 ]
 userVariables []
}
