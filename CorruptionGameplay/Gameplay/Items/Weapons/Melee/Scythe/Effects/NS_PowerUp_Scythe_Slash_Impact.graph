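# Niagara impact effect for the power-up scythe slash: a brief one-shot burst of sprites oriented around an impact direction.
# Glow flashes and slides along that direction, while Spike_big shoots along it, narrowing and stretching, both cooling from yellow toward dark purple.
# Slightly delayed Lazy_Sparks fly out with curl noise, drag, gravity and speed-based stretching.
# Glow also writes secondary glow color and size attributes, with alpha scaled by time of day.
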NiagaraSystem /CorruptionGameplay/Gameplay/Items/Weapons/Melee/Scythe/Effects/NS_PowerUp_Scythe_Slash_Impact
dependencies {
 usedDataInterfaces [
  &NiagaraDataInterfaceMeshRendererInfo
  &NiagaraDataInterfaceCurve
  &NiagaraDataInterfaceCurlNoise
  &NiagaraDataInterfaceColorCurve
  &NiagaraDataInterfaceVector2DCurve
 ]
 usedDynamicInputs [&NormalizeVector]
 usedModules [
  &self:SystemSpawnScript.SystemScriptSource.SystemScriptGraph.NiagaraNodeAssignment_0.SetVariables_#1
  &SystemState
 ]
}
emitters {
 Glow {
  emitterUpdateScript [
   &EmitterState(Life Cycle Mode=::NewEnumerator1, Loop Behavior=::NewEnumerator1, Loop Duration=1, UseLoopDelay=false)
   &SpawnBurst_Instantaneous(Spawn Count=1, Use Spawn Probability=false)
  ]
  particleSpawnScript [
   &InitializeParticle(Lifetime=0.2, Sprite Size Mode=::NewEnumerator3, Uniform Sprite Size=300, UsePositionOffset=false, Write Material Random=true)
   &CameraOffset(Camera Offset Amount=75)
   &DynamicMaterialParameters(
    Index 0 Param 1=1
    Index 0 Param 2=1
    Module.Param0WriteEnabled=true
    Module.Param1WriteEnabled=true
    Module.Param2WriteEnabled=true
    Module.Param3WriteEnabled=true
   )
  ]
  particleUpdateScript [
   &ParticleState()
   /SetVariables(Particles.Position=/Niagara/DynamicInputs/Transforms/TransformPosition(
     Position=&Multiply_VectorByFloat(Vector=$System.ImpactDirection, Float=&Lerp_Float(A=-30, B=45, Alpha=$Particles.NormalizedAge))
     Source Space=::Local
     Destination Space=::Simulation
    )) {isSetParameters=true}
   &ScaleSpriteSize(Scale Factor=&Vector2DFromFloat(Value=&FloatFromCurve(
      FloatCurve=DataInterface() {propertyValues {Curve {keys [
          {arriveTangent=13.1625, interpMode=RCIM_Cubic, leaveTangent=13.1625, tangentMode=RCTM_User, time=0, value=0}
          {arriveTangent=0.0473, interpMode=RCIM_Cubic, leaveTangent=0.0473, tangentMode=RCTM_User, time=0.1, value=1}
          {arriveTangent=-0.0078, interpMode=RCIM_Cubic, leaveTangent=-0.0078, tangentMode=RCTM_User, time=0.4, value=1}
          {arriveTangent=-2.5609, interpMode=RCIM_Cubic, leaveTangent=-2.5609, tangentMode=RCTM_User, time=1, value=0}
         ]}}}
      Scale Curve=&RandomRangeFloat(Minimum=0.7, Maximum=1, Evaluation Type=::NewEnumerator1)
     )))
   &ScaleColor(
    Linear Color Curve=DataInterface() {propertyValues {
      AlphaCurve {keys [<at 0 1 RCIM_Cubic>]}
      BlueCurve {keys [<at 0 0.2>, {arriveTangent=0.6172, interpMode=RCIM_Cubic, leaveTangent=0.6172, time=0.2134, value=0.0926}, <at 0.486 0.5>, <at 1 0.15>]}
      GreenCurve {keys [<at 0 0.9492>, {arriveTangent=-1.9529, interpMode=RCIM_Cubic, leaveTangent=-1.9529, time=0.2134, value=0.2944}, <at 0.486 0>, <at 1 0>]}
      RedCurve {keys [<at 0 1>, {arriveTangent=-1.4343, interpMode=RCIM_Cubic, leaveTangent=-1.4343, time=0.2134, value=0.75}, <at 0.486 0.3029>, <at 1 0.0908>]}
     }}
    Scale Mode=::NewEnumerator2
   ) {module=ScaleColor001}
   &ScaleColor(
    Color Value To Scale=$Particles.Color
    Scale Alpha=&FloatFromCurve(
     FloatCurve=DataInterface() {propertyValues {Curve {keys [
         {arriveTangent=0.4005, interpMode=RCIM_Cubic, leaveTangent=0.4005, tangentMode=RCTM_User, time=0, value=0}
         {arriveTangent=0.0256, interpMode=RCIM_Cubic, leaveTangent=0.0256, tangentMode=RCTM_User, time=0.1, value=1}
         {arriveTangent=0.01, interpMode=RCIM_Cubic, leaveTangent=0.01, tangentMode=RCTM_User, time=0.5, value=1}
         <at 1 0 RCIM_Cubic>
        ]}}}
     Scale Curve=0.75
    )
    Scale RGB=<5 5 5>
    ScaleA=true
    ScaleRGB=true
   )
   /SetVariables(
    Particles.GlowColor2:LinearColor=&MakeLinearColorFromVectorAndFloat(
     "Vector (RGB)"=&MakeVectorFromLinearColorRGB(Color=<rgba 3 1.1776 0.3703 1>)
     "Float (Alpha)"=&Multiply_Float(
      A=0.2
      B=&FloatFromCurve(
       FloatCurve=DataInterface() {propertyValues {Curve {keys [
           {arriveTangent=0.4005, interpMode=RCIM_Cubic, leaveTangent=0.4005, tangentMode=RCTM_User, time=0, value=0}
           {arriveTangent=0.0256, interpMode=RCIM_Cubic, leaveTangent=0.0256, tangentMode=RCTM_User, time=0.2, value=1}
           {arriveTangent=0.01, interpMode=RCIM_Cubic, leaveTangent=0.01, tangentMode=RCTM_User, time=0.7, value=1}
           <at 1 0 RCIM_Cubic>
          ]}}}
       Scale Curve=&FNTimeOfDayScalar(Night Scale=0.3, Morning Scale=0.2, Evening Scale=0.5)
      )
     )
    )
    Particles.GlowSize2:Vector2f=&Vector2DFromFloat(Value=&Multiply_Float(
      A=150
      B=&FloatFromCurve(
       FloatCurve=DataInterface() {propertyValues {Curve {keys [
           {arriveTangent=0.4346, interpMode=RCIM_Cubic, leaveTangent=0.4346, tangentMode=RCTM_User, time=0, value=0}
           {arriveTangent=0.0473, interpMode=RCIM_Cubic, leaveTangent=0.0473, tangentMode=RCTM_User, time=0.2, value=1}
           {arriveTangent=-0.0078, interpMode=RCIM_Cubic, leaveTangent=-0.0078, tangentMode=RCTM_User, time=0.5, value=1}
           <at 1 0 RCIM_Cubic>
          ]}}}
       Scale Curve=&RandomRangeFloat(Minimum=0.7, Maximum=1, Evaluation Type=::NewEnumerator1)
      )
     ))
   ) {isSetParameters=true}
  ]
  rendererClasses [&NiagaraSpriteRendererProperties]
  renderers [{rendererClass=&NiagaraSpriteRendererProperties}]
 }
 Lazy_Sparks {
  emitterUpdateScript [
   &EmitterState(Life Cycle Mode=::NewEnumerator1, Loop Behavior=::NewEnumerator1, Loop Delay=0.05, Loop Duration=1, UseLoopDelay=true)
   &SpawnBurst_Instantaneous(Spawn Count=&UniformRangedInt(Minimum=15, Maximum=25), Use Spawn Probability=false)
  ]
  particleSpawnScript [
   &InitializeParticle(
    Lifetime Max=0.35
    Lifetime Min=0.1
    Lifetime Mode=::NewEnumerator1
    Sprite Size Max=<2 2>
    Sprite Size Min=<1 1>
    Sprite Size Mode=::NewEnumerator2
    UsePositionOffset=false
    Write Material Random=true
   )
   &SphereLocation(
    MaskSpawnGroup=false
    Offset=&Multiply_VectorByFloat(Vector=$System.ImpactDirection, Float=&RandomRangeFloat(Minimum=-45, Maximum=30))
    Override Seed=false
    Sphere Radius=10
    Surface Only=false
    UseNonUniformScale=false
    UseOffset=true
    UseSphereOrientation=false
   )
   &AddVelocity(Velocity=&Multiply_VectorByFloat(Vector=$System.ImpactDirection, Float=&RandomRangeFloat(Minimum=150, Maximum=1000)))
   &AddVelocity(
    Coordinate Space=::Simulation
    Velocity=&Multiply_Vector(
     A=$Output.SphereLocation.SphereNormal
     B=&VectorFromFloat(Value=&UniformRangedFloat(Minimum=15, Maximum=200))
    )
   ) {module=AddVelocity001}
  ]
  particleUpdateScript [
   &ParticleState()
   &CameraOffset(Camera Offset Amount=75)
   &GravityForce(Gravity=&RandomRangeVector(Minimum=<0 0 0>, Maximum=<0 0 -100>))
   &CurlNoiseForce(
    Noise Frequency=20
    Noise Strength=&Multiply_Float(
     A=&RandomRangeFloat(Minimum=1000, Maximum=1500)
     B=&FloatFromCurve(FloatCurve=DataInterface() {propertyValues {Curve {keys [<at 0 0>,<at 0.2 1>]}}})
    )
    Pan Noise=true
   )
   &Drag(Drag=&RandomRangeFloat(Minimum=4, Maximum=8), Use Linear Drag=true, Use Rotational Drag=false)
   &SolveForcesAndVelocity(Clamp Velocity=false, Limit Acceleration=false)
   &ScaleColor(
    Linear Color Curve=DataInterface() {propertyValues {
      AlphaCurve {keys [<at 0 1 RCIM_Cubic>]}
      BlueCurve {keys [<at 0 1>, {interpMode=RCIM_Cubic, leaveTangent=-0.85, time=0.7675, value=1}]}
      GreenCurve {keys [<at 0 0>,<at 0.7675 0.125 RCIM_Cubic>]}
      RedCurve {keys [<at 0 0.7379>, {interpMode=RCIM_Cubic, leaveTangent=-0.6471, time=0.7675, value=0.2708}]}
     }}
    Scale Mode=::NewEnumerator2
   ) {module=ScaleColor001}
   &ScaleColor(
    Scale Alpha=&FloatFromCurve(FloatCurve=DataInterface() {propertyValues {Curve {keys [<at 0.66 1 RCIM_Cubic>,<at 1 0 RCIM_Cubic>]}}}, Scale Curve=0.5)
    Scale RGB=&VectorFromFloat(Value=&FloatFromCurve(
      FloatCurve=DataInterface() {propertyValues {Curve {keys [<at 0.5 2 RCIM_Cubic>,<at 1 0.5 RCIM_Cubic>]}}}
      Scale Curve=&FNTimeOfDayScalar(Night Scale=1, Morning Scale=1, Evening Scale=1)
     ))
    ScaleA=true
    ScaleRGB=true
   ) {module=ScaleColor002}
   &ScaleSpriteSizeBySpeed(Max Scale Factor=<0.25 5>, Min Scale Factor=<1 0.5>, Velocity Threshold=500)
   &ScaleSpriteSize(Scale Factor=&MakeVector2D(
     X=0.66
     Y=&FloatFromCurve(FloatCurve=DataInterface() {propertyValues {Curve {keys [<at 0 0.5 RCIM_Cubic>,<at 0.1 1.5 RCIM_Cubic>,<at 1 0.5>]}}})
    ))
  ]
  rendererClasses [&NiagaraSpriteRendererProperties]
  renderers [{rendererClass=&NiagaraSpriteRendererProperties}]
 }
 Spike_big {
  emitterUpdateScript [
   &EmitterState(Life Cycle Mode=::NewEnumerator1, Loop Behavior=::NewEnumerator1, Loop Duration=1, UseLoopDelay=false)
   &SpawnBurst_Instantaneous(Spawn Count=1, Use Spawn Probability=false)
  ]
  particleSpawnScript [
   &InitializeParticle(Lifetime=0.16, Sprite Size=<10 160>, Sprite Size Mode=::NewEnumerator4, UsePositionOffset=false, Write Material Random=true)
   &AddVelocity(Velocity=&Multiply_VectorByFloat(Vector=$System.ImpactDirection, Float=1000))
   &DynamicMaterialParameters(
    Index 0 Param 1=0.2
    Index 0 Param 2=1
    Module.Param0WriteEnabled=true
    Module.Param1WriteEnabled=true
    Module.Param2WriteEnabled=true
    Module.Param3WriteEnabled=true
   ) {enabled=false}
   /SetVariables(Particles.SubImageIndex=&RandomRangeFloat(Minimum=0, Maximum=3)) {isSetParameters=true}
  ]
  particleUpdateScript [
   &ParticleState()
   &CameraOffset(Camera Offset Amount=75)
   &SolveForcesAndVelocity(Clamp Velocity=false, Limit Acceleration=false)
   &ScaleColor(
    Scale Alpha=&FloatFromCurve(FloatCurve=DataInterface() {propertyValues {Curve {keys [<at 0.66 1 RCIM_Cubic>,<at 1 0 RCIM_Cubic>]}}}, Scale Curve=0.5)
    Scale RGB=&Multiply_Vector(
     A=<4 4 4>
     B=&MakeVectorFromLinearColorRGB(Color=&ColorFromCurve(ColorCurve=DataInterface() {propertyValues {
         AlphaCurve {keys [<at 0 1 RCIM_Cubic>]}
         BlueCurve {keys [<at 0 0.2>, {arriveTangent=-0.0603, interpMode=RCIM_Cubic, leaveTangent=-0.0603, time=0.426, value=0.0762}, <at 0.8297 0.15>]}
         GreenCurve {keys [<at 0 0.9492>, {arriveTangent=-1.144, interpMode=RCIM_Cubic, leaveTangent=-1.144, time=0.426, value=0.2423}, <at 0.8297 0>]}
         RedCurve {keys [<at 0 1>, {arriveTangent=-1.0958, interpMode=RCIM_Cubic, leaveTangent=-1.0958, time=0.426, value=0.6172}, <at 0.8297 0.0908>]}
        }}))
    )
    ScaleA=true
    ScaleRGB=true
   )
   &ScaleSpriteSize(Scale Factor=&Vector2DFromCurve(
     Vector2Curve=DataInterface() {propertyValues {
       XCurve {keys [<at 0 1 RCIM_Cubic>,<at 1 0.2 RCIM_Cubic>]}
       YCurve {keys [{arriveTangent=2.1214, interpMode=RCIM_Cubic, time=0, value=0.5}, <at 1 2 RCIM_Cubic>]}
      }}
     ScaleCurve=<1 1>
    ))
   &DynamicMaterialParameters(
    Index 0 Param 1=50
    Module.Param0WriteEnabled=true
    Module.Param1WriteEnabled=true
    Module.Param2WriteEnabled=true
    Module.Param3WriteEnabled=true
   ) {module=DynamicMaterialParameters001}
  ]
  rendererClasses [&NiagaraSpriteRendererProperties]
  renderers [{rendererClass=&NiagaraSpriteRendererProperties}]
 }
}
summary {
 systemName=NS_PowerUp_Scythe_Slash_Impact
 emitters [
  {emitterName=Lazy_Sparks, rendererClasses [&NiagaraSpriteRendererProperties]}
  {emitterName=Spike_big, rendererClasses [&NiagaraSpriteRendererProperties]}
  {emitterName=Glow, rendererClasses [&NiagaraSpriteRendererProperties]}
 ]
 userVariables [
  User.ImpactDirection {type=Vector3f, description="", defaultValue {struct=Vector3f, value=<0 1 0>}}
  User.SecondImpact {type=NiagaraFloat, defaultValue=0, description=""}
 ]
}
