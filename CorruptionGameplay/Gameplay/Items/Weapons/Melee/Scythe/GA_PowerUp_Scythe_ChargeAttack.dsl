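; Gameplay ability for the scythe's charge attack, a montage-driven dash that ends in a finisher or a miss animation.
; An intro notify starts continuous target selection, raced against a run timer and a net-synced primary-fire early cancel.
; Finisher montage notifies apply effect containers at a level set by weapon rarity.
; Jumping, falling, swimming or grinding ends the ability, and shared cleanup restores camera, rotation, tags and cues.
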(event Ability|EventActivateAbility
  ;[0 1 0.5389] Initialization/Activation
  ;;[1 1 1] Activate Ability
  (|CallParentFunction)
  (bind _returnvalue (Ability|CommitAbility))
  (if _returnvalue
    (Variables|Default|SetAbilityEnding false)
    (Variables|Default|SetFinisherCancelAllowed true)
    (CallFunction|InitAbility)
    (Ability|GameplayEffect|AddLooseGameplayTags (Ability|GetAvatarActorfromActorInfo) (Variables|NotifyTags|GetDashGrantedTags))
    (Camera|PopCameraMode)
    (Fortnite|KeepCGPathOptimizations (Variables|Refs|GetFortPlayerController))
    (CallFunction|DashIntro)
    (Ability|Tasks|WaitGameplayEvent "(TagName=\"Event.Jump.Pressed\")"
      (:then
        (Ability|Tasks|WaitGameplayTagAdd (Variables|NotifyTags|GetT_CubeBounce)
          (:then
            (Utilities|Time|SetTimerbyEvent (EventDispatchers|CreateEvent :!event K2_EndAbility) (Variables|Timers|GetOverallFailsafeTime))
            (Ability|Tasks|WaitGameplayTagAdd "(TagName=\"MovementMode.Sliding\")"
              (:Added
                (Ability|EndAbility))))
          (:Added
            (Ability|EndAbility))))
      (:EventReceived
        (Ability|EndAbility)))
    (else
      (Ability|EndAbility))))

(event Custom|WaitForPrimaryFire
  ;[0 1 0.5389] Initialization/Activation
  ;;[1 1 1] Primary Fire Listener
  (Ability|Tasks|WaitGameplayEvent "(TagName=\"Event.Primary.Pressed\")"
    (:then
      (if (Networking|IsServer)
        _))
    (:EventReceived
      _))
  (Ability|Tasks|WaitNetSync "OnlyServerWait"
    (:OnSync
      (if (Variables|Default|GetFinisherCancelAllowed)
        (Variables|Default|SetHitSomething true)
        (Variables|Default|SetFinisherAttackStarted true)
        (CallFunction|FinisherAttack (Utilities|Struct|MakeGameplayAbilityTargetDataHandle) true)))))

(event Custom|Dash
  ;[0 0.4317 1] Main Dash Executions
  ;;[1 1 1] Execute Dash
  (CallFunction|WaitforPrimaryFire)
  ;;; WaitForGamplayTags
  (Ability|Tasks|WaitGameplayTagAdd "(TagName=\"MovementMode.Falling\")"
    (:then
      (Ability|Tasks|WaitGameplayTagAdd "(TagName=\"Gameplay.Action.Player.SurfaceSwimming\")"
        (:then
          (Ability|Tasks|WaitGameplayTagAdd "(TagName=\"MovementMode.Grinding\")"
            (:then
              _)
            (:Added
              (Ability|EndAbility))) ; Need this to stop the charge effects when you start grinding
        )
        (:Added
          (Ability|EndAbility))))
    (:Added
      (Ability|EndAbility)))
  (Networking|ForceNetUpdate (Variables|Refs|GetPlayerPawn))
  (if (Variables|Default|GetAbilityEnding)
    _
    (elif (Networking|IsServer)
      (CallFunction|AddWeaponFX)
      (else
        _)))
  (bind (_asynctaskproxy _targetdata) (Ability|Tasks|DoTargetSelectionContinuous (Variables|Default|GetDashTargetSelectionTaskName) "(TagName=\"Gameplay.Melee.Hit1\")")
    (:then
      (bind _returnvalue (Utilities|Time|SetTimerbyEvent (EventDispatchers|CreateEvent :!event Custom|HitNothing) (Variables|Timers|GetRunTimer)))
      (Variables|Timers|SetRunTimerHandle _returnvalue)
      (Ability|SendGameplayEvent "(TagName=\"Event.Athena.Balloons.RemoveAll\")") ; Pop any balloons on the player to avoid \nyou attempting to charge as you float\n(the charge looks really bad when floating).
    )
    (:Targeted
      (Variables|Default|SetHitSomething true)
      (if (not (Variables|Default|GetFinisherAttackStarted))
        (Utilities|Time|ClearandInvalidateTimerbyHandle (Variables|Timers|GetRunTimerHandle))
        (CallFunction|PlayImpactGC _targetdata)
        (CallFunction|FinisherAttack _targetdata)))))

(event Custom|StartDashSoundCue
  ;[1 0 0.7611] FX/Audio
  ;;[1 1 1] Sound Cue Management
  (Ability|AddGameplayCueToOwner (Variables|Default|GetGC_DashLoopSound)))

(event Custom|EndDashSoundCue
  ;[1 0 0.7611] FX/Audio
  ;;[1 1 1] Sound Cue Management
  (Ability|RemoveGameplayCueFromOwner (Variables|Default|GetGC_DashLoopSound)))

(event Custom|AddWeaponFX
  ;[1 0 0.7611] FX/Audio
  ;;[1 1 1] Weapon FX
  (Development|PrintString "Adding scythe FX")
  (Ability|AddGameplayCueWithParamsToOwner (Variables|GameplayCues|GetGC_DashLoop) (Ability|GameplayCue|MakeGameplayCueParameters 0.0)))

(event Custom|DashIntro
  ;[0 0.4317 1] Main Dash Executions
  ;;[1 1 1] Dash Intro
  (bind (_asynctaskproxy _notifytag) (Ability|Tasks|PlayMontageAndWaitNotify "None" (Variables|Montages|GetM_DashIntro) (Variables|DashNumbers|GetDashPlayrate) "Intro")
    (:then
      (Input|SetForcedInputRotationSpeed (Variables|Refs|GetFortPlayerController) 0.75)
      (bind _asplayer_pawn_athena (Utilities|Casting|CastToPlayerPawn_Athena (Variables|Refs|GetPlayerPawn)))
      (Class|PlayerPawnAthena|StartAImRotationTick _asplayer_pawn_athena))
    (:OnNotifyBegin
      (if (GameplayTags|Equal(GameplayTag) _notifytag (Variables|NotifyTags|GetT_StartDashLoop))
        (CallFunction|Dash)
        (elif (GameplayTags|Equal(GameplayTag) _notifytag (Variables|NotifyTags|GetT_StartDashSoundCue))
          (CallFunction|StartDashSoundCue))))))

(event Custom|RemoveWeaponFX
  ;[1 0 0.7611] FX/Audio
  ;;[1 1 1] Weapon FX
  (Development|PrintString "Removing scythe FX")
  (Ability|RemoveGameplayCueFromOwner (Variables|GameplayCues|GetGC_DashLoop)))

(event Custom|FinisherAttack
  (params (InputPin :type Gameplay%20Ability%20Target%20Data%20Handle%20Structure%20%28by%20ref%29) (ForceHitPlayerAnim :type Boolean))
  ;[0.7426 1 0] Dash Outros/Finishers
  ;;[1 1 1] Finisher
  (Variables|Default|SetFinisherCancelAllowed false)
  (CallFunction|EndDashSoundCue)
  (Utilities|Time|ClearandInvalidateTimerbyHandle (Variables|Timers|GetRunTimerHandle))
  (Networking|ForceNetUpdate (Variables|Refs|GetPlayerPawn))
  (Ability|EndTaskbyInstanceName (Variables|Default|GetDashTargetSelectionTaskName))
  (Camera|PopCameraMode)
  (bind _montage (CallFunction|SelectHitMontage InputPin ForceHitPlayerAnim))
  (bind (_asynctaskproxy _notifytag) (Ability|Tasks|PlayMontageAndWaitNotify "None" _montage 1.0 "None" false)
    (:then
      (Input|SetRotatePawntoCamera (Variables|Refs|GetFortPlayerController) false)
      (Utilities|Time|SetTimerbyEvent (delegate Custom|CleanUpAfterHit) (+ (Animation|GetPlayLength _montage) (Variables|Timers|GetFinisherFailsafePadding))))
    (:OnCompleted
      _)
    (:OnBlendOut
      _)
    (:OnInterrupted
      _)
    (:OnCancelled
      _)
    (:OnNotifyBegin
      (CallFunction|HandleFinisherNotify _notifytag)))
  (CallFunction|CleanUpAfterHit))

(event Custom|CleanUpAfterHit
  ;[0.7426 1 0] Dash Outros/Finishers
  ;;[1 1 1] Finisher
  (Ability|EndAbility))

(event Custom|HandleFinisherNotify
  (params (Tag :type Gameplay%20Tag%20Structure))
  ;[0.7426 1 0] Dash Outros/Finishers
  ;;[1 1 1] Technically could handle any passed-in effect container tag, but specifically meant to handle the two hits of the finisher move
  (bind (_asynctaskproxy _targetdata _applicationtag) (Ability|Tasks|DoTargetSelection "None" Tag)
    (:Targeted
      (Ability|ApplyGameplayEffectContainer _targetdata _applicationtag (Math|Float|Truncate (Variables|EffectLevel|GetGELevel))))))

(event Custom|HitNothing
  ;[0.7426 1 0] Dash Outros/Finishers
  ;;[1 1 1] Hit Nothing
  (Variables|Default|SetFinisherCancelAllowed false)
  (Networking|ForceNetUpdate (Variables|Refs|GetPlayerPawn))
  (Ability|Tasks|PlayMontageAndWaitNotify "None" (Variables|Montages|GetMontageHitNothing) 1.0 "None" false
    (:then
      (Ability|EndTaskbyInstanceName (Variables|Default|GetDashTargetSelectionTaskName))
      (CallFunction|EndDashSoundCue))
    (:OnCompleted
      _)
    (:OnBlendOut
      _)
    (:OnInterrupted
      _)
    (:OnCancelled
      _))
  (Ability|EndAbility))

(event Custom|EarlyFinisher
  ;[0.7426 1 0] Dash Outros/Finishers
  ;;[1 1 1] Early Finisher -> Finisher move when cancelling early with primary attack
  ;;;[1 1 1] Do a singular target selection to see if our finisher hits something
  (bind (_asynctaskproxy _targetdata _applicationtag) (Ability|Tasks|DoTargetSelection "None" "(TagName=\"Gameplay.Melee.Hit1\")")
    (:then
      (CallFunction|FinisherAttack _targetdata true))
    (:Targeted
      (Variables|Default|SetHitSomething true)
      (Camera|PopCameraMode)
      (Ability|ApplyGameplayEffectContainer _targetdata _applicationtag (Math|Float|Truncate (Variables|EffectLevel|GetGELevel)))
      (CallFunction|PlayImpactGC _targetdata))))

(event Ability|EventOnEndAbility
  (params (bWasCancelled :type Boolean))
  ;[1 0.2581 0] Ability ending/cleanup
  (Variables|Default|SetAbilityEnding true)
  (Development|PrintString "Charge ability ended")
  (CallFunction|Cleanup)
  (|CallParentFunction bWasCancelled))

(event Custom|Cleanup
  ;[1 0.2581 0] Ability ending/cleanup
  ;;[1 1 1] Cleanup
  (Variables|Default|SetHitSomething false)
  (Networking|ForceNetUpdate (Variables|Refs|GetPlayerPawn))
  (CallFunction|RemoveWeaponFX)
  (CallFunction|CleanUpDash) ; tags to drive umbrella animation
  (Ability|RemoveGameplayCueFromOwner (Variables|GameplayCues|GetGC_DashLoop))
  (Ability|EndTaskbyInstanceName (Variables|Default|GetDashTargetSelectionTaskName))
  (bind _asplayer_pawn_athena (Utilities|Casting|CastToPlayerPawn_Athena (Variables|Refs|GetPlayerPawn))
    (:then
      (Class|PlayerPawnAthena|EndRotationTick _asplayer_pawn_athena))
    (:CastFailed
      _))
  (Input|ResetForcedInputRotationSpeed (Variables|Refs|GetFortPlayerController))
  (Input|SetRotatePawntoCamera (Variables|Refs|GetFortPlayerController) true)
  (CallFunction|EndDashSoundCue)
  (Camera|PopCameraMode)
  (Ability|GameplayEffect|RemoveLooseGameplayTags (Ability|GetAvatarActorfromActorInfo) (Variables|NotifyTags|GetDashGrantedTags))
  (Ability|Animation|MontageStop 0.5)
  (Variables|Default|SetFinisherCancelAllowed false))

(fn CleanUpDash)

(fn ImpulseHitDashTarget
  (params (TargetData :type Gameplay%20Ability%20Target%20Data%20Handle%20Structure)))

(fn InitAbility
  (bind _asfort_player_pawn (Utilities|Casting|CastToFortPlayerPawn (Ability|GetAvatarActorfromActorInfo)))
  (bind _output_get (Variables|Refs|SetPlayerPawn _asfort_player_pawn))
  (bind _asplayer_pawn_athena (Utilities|Casting|CastToPlayerPawn_Athena _output_get))
  (Variables|Refs|SetPlayerPawnAthena _asplayer_pawn_athena)
  (Variables|Refs|SetFortPlayerController (Fortnite|GetFortPlayerControllerfromActor (Ability|GetAvatarActorfromActorInfo)))
  (if (Tags|WeaponDataHasMatchingGameplayTag (Variables|/Script/FortniteGame.FortPawn|GetCurrentWeapon _output_get) (Variables|EffectLevel|GetRarityUncommon))
    (Variables|EffectLevel|SetAdditionalLevels 1.0)
    (elif (Tags|WeaponDataHasMatchingGameplayTag (Variables|/Script/FortniteGame.FortPawn|GetCurrentWeapon _output_get) (Variables|EffectLevel|GetRarityRare))
      (Variables|EffectLevel|SetAdditionalLevels 2.0)
      (elif (Tags|WeaponDataHasMatchingGameplayTag (Variables|/Script/FortniteGame.FortPawn|GetCurrentWeapon _output_get) (Variables|EffectLevel|GetRarityEpic))
        (Variables|EffectLevel|SetAdditionalLevels 3.0)
        (elif (Tags|WeaponDataHasMatchingGameplayTag (Variables|/Script/FortniteGame.FortPawn|GetCurrentWeapon _output_get) (Variables|EffectLevel|GetRarityLegendary))
          (Variables|EffectLevel|SetAdditionalLevels 4.0)
          (elif (Tags|WeaponDataHasMatchingGameplayTag (Variables|/Script/FortniteGame.FortPawn|GetCurrentWeapon _output_get) (Variables|EffectLevel|GetRarityMythic))
            (Variables|EffectLevel|SetAdditionalLevels 5.0)
            (else
              _))))))
  (Variables|EffectLevel|SetGELevel (+ (Variables|EffectLevel|GetDefaultLevel) (Variables|EffectLevel|GetAdditionalLevels)))
  (return))

(fn OnRep_AdditionalLevels)

(fn OnRep_ChargeInterval)

(fn OnRep_ChargeTime)

(fn PlayImpactGC
  (params (TargetData :type Gameplay%20Ability%20Target%20Data%20Handle%20Structure))
  (bind (_bblockinghit _binitialoverlap _time _distance _location _impactpoint _normal) (Collision|BreakHitResult (Ability|TargetData|GetHitResultfromTargetData TargetData)))
  (Ability|ExecuteGameplayCueWithParamsOnOwner (Variables|GameplayCues|GetGC_DashImpact) (Ability|GameplayCue|MakeGameplayCueParameters :NormalizedMagnitude 0.0 :RawMagnitude 0.0 :EffectContext (Ability|GetContextfromOwner TargetData) :Location _location :Normal _normal :GameplayEffectLevel 1 :AbilityLevel 1 :bReplicateLocationWhenUsingMinimalRepProxy false)))

(fn SelectHitMontage
  (params (TargetData :type Gameplay%20Ability%20Target%20Data%20Handle%20Structure) (ForceHitPlayer :type Boolean))
  (bind _returnvalue (Ability|TargetData|GetHitResultfromTargetData TargetData))
  (bind (_bblockinghit _binitialoverlap _time _distance _location _impactpoint _normal _impactnormal _physmat _hitactor) (Collision|BreakHitResult _returnvalue))
  (if ForceHitPlayer
    (return (Variables|Montages|GetM_HitPlayer))
    (else
      (Utilities|IsValid _hitactor
        (:Is%20Valid
          (Development|PrintString (Utilities|String|Append "Hit " (Utilities|GetDisplayName _hitactor) ", can be damaged: " (Utilities|String|ToString(Boolean) (Variables|/Script/Engine.Actor|GetbCanBeDamaged _hitactor))))
          (bind _shouldignore (CallFunction|ShouldIgnoreDamageableActor _hitactor))
          (if _shouldignore
            (return (Variables|Montages|GetM_HitWorld))
            (else
              (return (select (Variables|/Script/Engine.Actor|GetbCanBeDamaged _hitactor) (Variables|Montages|GetM_HitPlayer) (Variables|Montages|GetM_HitWorld))))))
        (:Is%20Not%20Valid
          (return (Variables|Montages|GetM_HitWorld)))))))

(fn ShouldIgnoreDamageableActor
  (params (Actor :type Actor%20Object%20Reference))
  (Utilities|Casting|CastToBGA_Athena_CorruptionGameplay_Cube Actor)
  (return true))
