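; Gameplay ability for the Mantis scythe melee power-up, a child class that defers activation to its parent.
; On end, it clears the aim-assist targeting timer and prints a debug string, with no parent call in that event.
; It overrides the custom ability source transform to a point a fixed distance behind the avatar along its forward vector, keeping the avatar's rotation and using unit scale.
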(event Ability|EventActivateAbility
  (|CallParentFunction))

(event Ability|EventOnEndAbility
  (params (bWasCancelled :type Boolean))
  (Utilities|Time|ClearandInvalidateTimerbyHandle (Variables|Default|GetAimAssistTargetingTimerHandle))
  (Development|PrintString "Scythe ability ernd" false))

(fn GetCustomAbilitySourceTransform
  (bind (_returnvalue_location _returnvalue_rotation) (Math|Transform|BreakTransform (Transformation|GetActorTransform (Ability|GetAvatarActorfromActorInfo))))
  (return (- _returnvalue_location (* (Transformation|GetActorForwardVector (Ability|GetAvatarActorfromActorInfo)) 112.0)) _returnvalue_rotation "1.000000,1.000000,1.000000"))
