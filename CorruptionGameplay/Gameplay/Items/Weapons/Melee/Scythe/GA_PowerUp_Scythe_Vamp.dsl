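; Scythe vampirism power-up ability that siphons health and shield to the owner on scythe events against targets other than buildings and vehicles.
; GetMagnitudes picks a player or AI (Deimos-tagged target) siphon amount, filling missing health first and spilling the excess into shield.
; Both values feed one SetByCaller effect, applied only if either is positive.
; InitAbility also derives weapon-rarity levels, but GetMagnitudes does not read them.
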(event Ability|EventActivateAbilityFromEvent
  (params (EventData :type Gameplay%20Event%20Data%20Structure%20%28by%20ref%29))
  (bind (_eventtag _instigator _target _optionalobject _optionalobject2 _contexthandle _instigatortags _targettags) (Utilities|Struct|BreakGameplayEventData EventData))
  (bind (_healmagnitude _shieldmagnitude _applyeffect) (CallFunction|GetMagnitudes))
  (|CallParentFunction EventData)
  (Variables|Tags|SetTarget%20Tags _targettags)
  (CallFunction|InitAbility)
  (bind _returnvalue (Ability|GameplayEffect|AssignTagSetbyCallerMagnitude (Ability|MakeOutgoingGameplayEffectSpec (Variables|GameplayEffects|GetGE_HealthShield)) "(TagName=\"SetByCaller.Healing\")" _healmagnitude))
  (bind _returnvalue_2 (Ability|GameplayEffect|AssignTagSetbyCallerMagnitude _returnvalue "(TagName=\"SetByCaller.Shield\")" _shieldmagnitude))
  (if _applyeffect
    (Ability|ApplyGameplayEffectSpecToOwner _returnvalue_2)))

(fn GetMagnitudes
  (return (Math|Float|Clamp(Float) (ScalableFloat|GetValueatLevel (select (Variables|Bools|GetIsDeimos) (Variables|HotFixValues|GetAISiphonAmount) (Variables|HotFixValues|GetPlayerSiphonAmount))) 0.0 (* (Math|Conversions|ToFloat(Integer) (Math|Float|Round (* (- (FortPawn|GetMaxHealth (Variables|Pawn|GetOwningActor)) (FortPawn|GetHealth (Variables|Pawn|GetOwningActor))) 1e+02))) 0.01)) (Math|Float|Clamp(Float) (Math|Float|Clamp(Float) (- (ScalableFloat|GetValueatLevel (select (Variables|Bools|GetIsDeimos) (Variables|HotFixValues|GetAISiphonAmount) (Variables|HotFixValues|GetPlayerSiphonAmount))) (* (Math|Conversions|ToFloat(Integer) (Math|Float|Round (* (- (FortPawn|GetMaxHealth (Variables|Pawn|GetOwningActor)) (FortPawn|GetHealth (Variables|Pawn|GetOwningActor))) 1e+02))) 0.01)) 0.0 1e+04) 0.0 (* (Math|Conversions|ToFloat(Integer) (Math|Float|Round (* (- (FortPawn|GetMaxShield (Variables|Pawn|GetOwningActor)) (FortPawn|GetShield (Variables|Pawn|GetOwningActor))) 1e+02))) 0.01)) (or (> (Math|Float|Clamp(Float) (ScalableFloat|GetValueatLevel (select (Variables|Bools|GetIsDeimos) (Variables|HotFixValues|GetAISiphonAmount) (Variables|HotFixValues|GetPlayerSiphonAmount))) 0.0 (* (Math|Conversions|ToFloat(Integer) (Math|Float|Round (* (- (FortPawn|GetMaxHealth (Variables|Pawn|GetOwningActor)) (FortPawn|GetHealth (Variables|Pawn|GetOwningActor))) 1e+02))) 0.01)) 0.0) (> (Math|Float|Clamp(Float) (Math|Float|Clamp(Float) (- (ScalableFloat|GetValueatLevel (select (Variables|Bools|GetIsDeimos) (Variables|HotFixValues|GetAISiphonAmount) (Variables|HotFixValues|GetPlayerSiphonAmount))) (* (Math|Conversions|ToFloat(Integer) (Math|Float|Round (* (- (FortPawn|GetMaxHealth (Variables|Pawn|GetOwningActor)) (FortPawn|GetHealth (Variables|Pawn|GetOwningActor))) 1e+02))) 0.01)) 0.0 1e+04) 0.0 (* (Math|Conversions|ToFloat(Integer) (Math|Float|Round (* (- (FortPawn|GetMaxShield (Variables|Pawn|GetOwningActor)) (FortPawn|GetShield (Variables|Pawn|GetOwningActor))) 1e+02))) 0.01)) 0.0))))

(fn InitAbility
  (bind _additionallevels (Variables|Float|GetAdditionalLevels))
  (bind _defaultlevel (Variables|Float|GetDefaultLevel))
  (bind _returnvalue (+ _defaultlevel _additionallevels))
  (bind _playersiphonamount (Variables|HotFixValues|GetPlayerSiphonAmount))
  (bind _returnvalue_2 (ScalableFloat|GetValueatLevel _playersiphonamount _returnvalue))
  (bind _output_get (Variables|Pawn|SetOwningActor (Ability|GetActivatingPawn)))
  (if (GameplayTags|HasTag (Variables|Tags|GetTarget%20Tags) (Variables|Tags|GetDeimosTag))
    (Variables|Bools|SetIsDeimos true)
    (else
      _))
  (if (Tags|WeaponDataHasMatchingGameplayTag (Variables|/Script/FortniteGame.FortPawn|GetCurrentWeapon _output_get) (Variables|Tags|GetRarityUncommon))
    (Variables|Float|SetAdditionalLevels 1.0)
    (elif (Tags|WeaponDataHasMatchingGameplayTag (Variables|/Script/FortniteGame.FortPawn|GetCurrentWeapon _output_get) (Variables|Tags|GetRarityRare))
      (Variables|Float|SetAdditionalLevels 2.0)
      (elif (Tags|WeaponDataHasMatchingGameplayTag (Variables|/Script/FortniteGame.FortPawn|GetCurrentWeapon _output_get) (Variables|Tags|GetRarityEpic))
        (Variables|Float|SetAdditionalLevels 3.0)
        (elif (Tags|WeaponDataHasMatchingGameplayTag (Variables|/Script/FortniteGame.FortPawn|GetCurrentWeapon _output_get) (Variables|Tags|GetRarityLegendary))
          (Variables|Float|SetAdditionalLevels 4.0)
          (elif (Tags|WeaponDataHasMatchingGameplayTag (Variables|/Script/FortniteGame.FortPawn|GetCurrentWeapon _output_get) (Variables|Tags|GetRarityMythic))
            (Variables|Float|SetAdditionalLevels 5.0)
            (else
              _))))))
  (Variables|Float|SetHealthMultiplier _returnvalue_2)
  (Variables|Float|SetHealthAmount _returnvalue_2))

(fn ShouldAbilityRespondToEvent
  (params (ActorInfo :type Gameplay%20Ability%20Actor%20Info%20Structure) (Payload :type Gameplay%20Event%20Data%20Structure))
  (bind (_eventtag _instigator _target _optionalobject _optionalobject2 _contexthandle _instigatortags) (Utilities|Struct|BreakGameplayEventData Payload))
  (if (GameplayTags|HasAnyTags _instigatortags (Variables|Tags|GetTScythe))
    (Utilities|Casting|CastToBuildingActor _target
      (:then
        (return false))
      (:CastFailed
        (bind _returnvalue (GameplayTags|HasMatchingGameplayTag(Message) _target (Variables|Tags|GetT_Vehicle)))
        (if _returnvalue
          (return false)
          (else
            (return true)))))
    (else
      (return false))))
